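# Idle goop effect for the Psyche character's backpack, emitting small mesh blobs that drift upward, tumble, swell and shrink away.
# Only Sludge_Optimized is enabled: two spawn groups index an offset array to feed both left and right points from one emitter, replacing disabled per-side, socket-based emitters.
# Blobs push away from an origin point with slight opposite-to-owner velocity, rise under positive-Z gravity, and animate a dynamic material parameter over life.
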NiagaraSystem /BRCosmetics/Characters/Player/Male/Medium/Bodies/M_MED_Psyche/FX/NS_Psyche_Goop_Backpack_Idle
dependencies {
 usedDataInterfaces [&NiagaraDataInterfaceSkeletalMesh, &NiagaraDataInterfaceCurve, &NiagaraDataInterfaceArrayFloat3]
 usedDynamicInputs []
 usedModules [&SystemState]
}
emitters {
 Sludge_Base {
  bEnabled=false
  emitterUpdateScript [
   &EmitterState(Inactive Response=::NewEnumerator1, Life Cycle Mode=::NewEnumerator1, Loop Duration=2, UseLoopDelay=false)
   &SpawnRate(
    SpawnRate=&RandomRangeFloat(Minimum=3, Maximum=5, Evaluation Type=::NewEnumerator1)
    Use Spawn Probability=false
   )
  ]
  particleSpawnScript [
   &InitializeParticle(
    Lifetime Mode=::NewEnumerator1
    Mesh Scale Max=<1 0.8 1.5>
    Mesh Scale Min=<0.2 0.2 0.2>
    Mesh Scale Mode=::NewEnumerator2
    Position Offset=<-40 0 0>
    UsePositionOffset=true
    Write Material Random=true
   )
   &Color(Color=&MakeLinearColorFromVectorAndFloat("Vector (RGB)"=<1 1 1>, "Float (Alpha)"=&ClampFloat(Float=1))) {enabled=false}
   &DynamicMaterialParameters(
    Index 0 Param 1=2
    Index 0 Param 2=10
    Index 0 Param 3=3
    Index 0 Param 4=&RandomRangeFloat(Minimum=50, Maximum=55, Evaluation Type=::NewEnumerator1)
    Module.Param0WriteEnabled=true
    Module.Param1WriteEnabled=true
    Module.Param2WriteEnabled=true
    Module.Param3WriteEnabled=false
   ) {module=DynamicMaterialParameters002}
   &SkeletalMeshLocation(
    MaskSpawnGroup=false
    Mesh Sampling Type=::NewEnumerator5
    OverrideRandomSeed=false
    Skeletal Mesh=DataInterface() {propertyValues {
      DefaultMesh=null
      ExcludeBoneName=root
      OptionalComponentClass=null
      PreviewMesh=/BRCosmetics/Accessories/FORT_Backpacks/Backpack_M_MED_Psyche/Meshes/M_MED_Psyche_Pack
      SkinningMode=SkinOnTheFly
      SoftSourceActor=null
      SourceMode=Default
      UvSetIndex=0
      WholeMeshLOD=-1
      bExcludeBone=false
      bReadDeformedGeometry=true
      bRequireCurrentFrameData=true
      ComponentTags []
      FilteredBones []
      FilteredSockets [FX_RightGoop]
      MeshUserParameter {parameter {name=None, typeDefHandle {registeredTypeIndex=1}, varData []}}
      SamplingRegions []
     }}
   ) {enabled=false}
   &AddVelocity(Coordinate Space=::World, Velocity=<0 0 10>) {module=AddVelocity001}
  ]
  particleUpdateScript [
   &ParticleState()
   &ScaleMeshSize(Scale Factor=&VectorFromFloat(Value=&FloatFromCurve(FloatCurve=DataInterface() {propertyValues {Curve {keys [
          {arriveTangent=9.1187, interpMode=RCIM_Cubic, time=0, value=0}
          {arriveTangent=0.3759, interpMode=RCIM_Cubic, leaveTangent=0.3759, time=0.2487, value=1.1246}
          {arriveTangent=-1.497, interpMode=RCIM_Cubic, leaveTangent=-1.497, time=0.7835, value=0.2945}
          <at 1 0 RCIM_Cubic>
         ]}}}))) {module=ScaleMeshSize001}
   &ScaleColor(
    Scale Alpha=&FloatFromCurve(FloatCurve=DataInterface() {propertyValues {Curve {keys [<at 0.5 1.0113 RCIM_Cubic>,<at 1.0011 0.002 RCIM_Cubic>]}}})
    ScaleA=true
    ScaleRGB=false
   ) {enabled=false}
   &DynamicMaterialParameters(
    Index 0 Param 1=0.1
    Index 0 Param 3=&FloatFromCurve(FloatCurve=DataInterface() {propertyValues {Curve {keys [<at 0 3 RCIM_Cubic>, {arriveTangent=-4.1151, interpMode=RCIM_Cubic, leaveTangent=-4.1151, tangentMode=RCTM_User, time=1, value=1}]}}})
    Module.Param0WriteEnabled=false
    Module.Param1WriteEnabled=false
    Module.Param2WriteEnabled=true
    Module.Param3WriteEnabled=false
   )
   &AccelerationForce(Acceleration=&Multiply_VectorByFloat(
     Vector=&RandomRangeVector(Minimum=<-3 -3 5>, Maximum=<3 3 10>, Evaluation Type=::NewEnumerator1)
     Float=1
    ))
   &Drag(Drag=1.5, Use Linear Drag=true, Use Rotational Drag=false) {enabled=false}
   &SolveForcesAndVelocity(Clamp Velocity=false, Limit Acceleration=false)
   &MeshRotationForce(
    Coordinate Space:ENiagaraMeshTransforms=::NewEnumerator3
    Rotation=&RandomRangeVector(Minimum=<-4 -4 -4>, Maximum=<4 4 4>)
   ) {enabled=false}
   &SolveRotationalForcesAndVelocity() {enabled=false}
  ]
  rendererClasses [&NiagaraMeshRendererProperties]
  renderers [{rendererClass=&NiagaraMeshRendererProperties, rendererIndex=0}]
 }
 Sludge_Base001 {
  bEnabled=false
  emitterUpdateScript [
   &EmitterState(Inactive Response=::NewEnumerator1, Life Cycle Mode=::NewEnumerator1, Loop Duration=2, UseLoopDelay=false)
   &SpawnRate(SpawnRate=4.3, Use Spawn Probability=false)
  ]
  particleSpawnScript [
   &InitializeParticle(
    Lifetime Max=1.6
    Lifetime Min=1.3
    Lifetime Mode=::NewEnumerator1
    Mesh Scale Max=<1 0.5 0.8>
    Mesh Scale Min=<0.2 0.2 0.6>
    Mesh Scale Mode=::NewEnumerator2
    Position Mode=::NewEnumerator1
    Position Offset=<0 0 -4>
    UsePositionOffset=true
    Write Material Random=true
   )
   &Color(Color=&MakeLinearColorFromVectorAndFloat("Vector (RGB)"=<1 1 1>, "Float (Alpha)"=&ClampFloat(Float=1))) {enabled=false}
   &DynamicMaterialParameters(
    Index 0 Param 1=2
    Index 0 Param 2=10
    Index 0 Param 3=3
    Index 0 Param 4=&RandomRangeFloat(Minimum=50, Maximum=55, Evaluation Type=::NewEnumerator1)
    Module.Param0WriteEnabled=true
    Module.Param1WriteEnabled=true
    Module.Param2WriteEnabled=true
    Module.Param3WriteEnabled=false
   ) {module=DynamicMaterialParameters002}
   &SkeletalMeshLocation(
    MaskSpawnGroup=false
    Mesh Sampling Type=::NewEnumerator5
    OverrideRandomSeed=false
    Skeletal Mesh=DataInterface() {propertyValues {
      DefaultMesh=null
      ExcludeBoneName=root
      OptionalComponentClass=null
      PreviewMesh=&M_MED_Psyche_Pack
      SkinningMode=SkinOnTheFly
      SoftSourceActor=null
      SourceMode=Default
      UvSetIndex=0
      WholeMeshLOD=-1
      bExcludeBone=false
      bReadDeformedGeometry=true
      bRequireCurrentFrameData=true
      ComponentTags []
      FilteredBones []
      FilteredSockets [FX_LeftGoop]
      MeshUserParameter {parameter {name=None, typeDefHandle {registeredTypeIndex=1}, varData []}}
      SamplingRegions []
     }}
   ) {enabled=false}
   &AddVelocity(Coordinate Space=::World, Velocity=&RandomRangeVector(Minimum=<-1 -1 20>, Maximum=<1 1 15>)) {module=AddVelocity001}
   &InitialMeshOrientation(Use Rotation Vector=false) {enabled=false}
  ]
  particleUpdateScript [
   &ParticleState()
   &ScaleMeshSize(Scale Factor=&VectorFromFloat(Value=&FloatFromCurve(FloatCurve=DataInterface() {propertyValues {Curve {keys [
          {arriveTangent=9.1187, interpMode=RCIM_Cubic, time=0, value=0}
          {arriveTangent=0.1274, interpMode=RCIM_Cubic, leaveTangent=0.1274, tangentMode=RCTM_User, time=0.2487, value=1.1246}
          {arriveTangent=-2.4086, interpMode=RCIM_Cubic, leaveTangent=-2.4086, tangentMode=RCTM_User, time=1, value=0}
         ]}}}))) {module=ScaleMeshSize001}
   &ScaleColor(
    Scale Alpha=&FloatFromCurve(FloatCurve=DataInterface() {propertyValues {Curve {keys [<at 0.5 1.0113 RCIM_Cubic>,<at 1.0011 0.002 RCIM_Cubic>]}}})
    ScaleA=true
    ScaleRGB=false
   ) {enabled=false}
   &DynamicMaterialParameters(
    Index 0 Param 1=0.1
    Index 0 Param 3=&FloatFromCurve(FloatCurve=DataInterface() {propertyValues {Curve {keys [<at 0 3 RCIM_Cubic>, {arriveTangent=-4.1151, interpMode=RCIM_Cubic, leaveTangent=-4.1151, tangentMode=RCTM_User, time=1, value=1}]}}})
    Module.Param0WriteEnabled=false
    Module.Param1WriteEnabled=false
    Module.Param2WriteEnabled=true
    Module.Param3WriteEnabled=false
   ) {enabled=false}
   &AccelerationForce(
    Acceleration=&Multiply_VectorByFloat(Vector=&RandomRangeVector(Minimum=<-1.5 -1.5 7>, Maximum=<1.5 1.5 9>), Float=1)
    Coordinate Space=::Simulation
   ) {enabled=false}
   &Drag(Drag=1.5, Use Linear Drag=true, Use Rotational Drag=false) {enabled=false}
   &SolveForcesAndVelocity(Clamp Velocity=false, Limit Acceleration=false)
   &MeshRotationForce(
    Coordinate Space:ENiagaraMeshTransforms=::NewEnumerator3
    Rotation=&RandomRangeVector(Minimum=<-4 -4 -4>, Maximum=<4 4 4>)
   )
   &SolveRotationalForcesAndVelocity()
  ]
  rendererClasses [&NiagaraMeshRendererProperties]
  renderers [{rendererClass=&NiagaraMeshRendererProperties, rendererIndex=0}]
 }
 Sludge_Optimized {
  emitterUpdateScript [
   &EmitterState(Inactive Response=::NewEnumerator1, Life Cycle Mode=::NewEnumerator1, Loop Duration=2, UseLoopDelay=false)
   &SpawnRate(SpawnRate=6, Use Spawn Probability=false)
   &SpawnRate(SpawnGroup=1, SpawnRate=6, Use Spawn Probability=false) {module=SpawnRate001}
  ]
  particleSpawnScript [
   &InitializeParticle(
    Lifetime Max=1.8
    Lifetime Min=1.5
    Lifetime Mode=::NewEnumerator1
    Mesh Scale Max=<1 0.5 0.8>
    Mesh Scale Min=<0.2 0.2 0.6>
    Mesh Scale Mode=::NewEnumerator2
    Position Offset=<0 0 -4>
    UsePositionOffset=true
    Write Material Random=true
   )
   &DynamicMaterialParameters(
    Index 0 Param 1=2
    Index 0 Param 2=10
    Index 0 Param 3=3
    Index 0 Param 4=&RandomRangeFloat(Minimum=50, Maximum=55, Evaluation Type=::NewEnumerator1)
    Module.Param0WriteEnabled=true
    Module.Param1WriteEnabled=true
    Module.Param2WriteEnabled=true
    Module.Param3WriteEnabled=false
   ) {module=DynamicMaterialParameters002}
   &BoxLocation(
    Box Offset=&SelectVectorFromArray(Direct Array Index=$Emitter.SpawnGroup, Vector Selection Array=DataInterface() {propertyValues {FloatData [<11 -6.1 5>, <-11 -6.1 5>]}})
    Box Size=<0 0 0>
    MaskSpawnGroup=false
    Module.Override Seed=false
    Rotate Box=false
    Surface Only=false
   )
   &AddVelocityFromPoint(Module.Use Falloff Distance=false, Origin Offset=<0 -5 2.5>, Velocity Strength=6)
   &AddVelocity(
    Coordinate Space=::World
    Velocity=&Add_Vector(A=&Multiply_VectorByFloat(Vector=$Engine.Owner.Velocity, Float=-0.01), B=<0 0 0>)
   )
  ]
  particleUpdateScript [
   &ParticleState()
   &DynamicMaterialParameters(
    Index 0 Param 3=&FloatFromCurve(
     FloatCurve=DataInterface() {propertyValues {Curve {keys [
         <at 0 0 RCIM_Cubic>
         {arriveTangent=2.4962, interpMode=RCIM_Cubic, leaveTangent=2.4962, tangentMode=RCTM_User, time=0.3, value=0.45}
         <at 1 1>
        ]}}}
     Scale Curve=2
    )
    Module.Param0WriteEnabled=false
    Module.Param1WriteEnabled=false
    Module.Param2WriteEnabled=true
    Module.Param3WriteEnabled=false
   )
   &ScaleMeshSize(Scale Factor=&VectorFromFloat(Value=&FloatFromCurve(FloatCurve=DataInterface() {propertyValues {Curve {keys [
          {arriveTangent=9.1187, interpMode=RCIM_Cubic, time=0, value=0}
          {arriveTangent=0.1274, interpMode=RCIM_Cubic, leaveTangent=0.1274, tangentMode=RCTM_User, time=0.2487, value=1.1246}
          {arriveTangent=-3.4821, interpMode=RCIM_Cubic, leaveTangent=-3.4821, tangentMode=RCTM_User, time=1, value=0}
         ]}}}))) {module=ScaleMeshSize001}
   &GravityForce(Gravity=<0 0 16>)
   &MeshRotationForce(
    Coordinate Space:ENiagaraMeshTransforms=::NewEnumerator3
    Rotation=&RandomRangeVector(Minimum=<-4 -4 -4>, Maximum=<4 4 4>)
   )
   &SolveForcesAndVelocity(Clamp Velocity=false, Limit Acceleration=false)
   &SolveRotationalForcesAndVelocity()
  ]
  rendererClasses [&NiagaraMeshRendererProperties]
  renderers [
   {rendererClass=&NiagaraMeshRendererProperties, rendererIndex=0}
   {rendererClass=&NiagaraMeshRendererProperties, rendererIndex=1}
  ]
 }
}
summary {
 systemName=NS_Psyche_Goop_Backpack_Idle
 emitters [
  {bEnabled=false, emitterName=Sludge_Base001, rendererClasses [&NiagaraMeshRendererProperties]}
  {bEnabled=false, emitterName=Sludge_Base, rendererClasses [&NiagaraMeshRendererProperties]}
  {emitterName=Sludge_Optimized, rendererClasses [&NiagaraMeshRendererProperties]}
 ]
 userVariables [
  User.bIsFrontEnd {type=NiagaraBool}
  User.bIsFrontEndPreview {type=NiagaraBool}
  User.SlurpColorAlpha {type=NiagaraFloat, defaultValue=0, description=""}
 ]
}
