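# Niagara idle effect for the Psyche character cosmetic body, emitting sludge mesh particles from the skeletal mesh's FX_Goo sockets.
# One emitter scales with owner speed: faster movement lowers spawn rate, shortens lifetime and increases upward world velocity.
# Particles drift upward with random acceleration and tumble, while a curve swells then shrinks them.
# Dynamic material parameters drive the look, one animated over life; color, alpha-fade and drag modules are disabled.
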NiagaraSystem /BRCosmetics/Characters/Player/Male/Medium/Bodies/M_MED_Psyche/FX/NS_Psyche_Goop_Idle
dependencies {usedDataInterfaces [&NiagaraDataInterfaceCurve], usedDynamicInputs [], usedModules [&SystemState]}
emitters {Sludge_Base001 {
  emitterUpdateScript [
   &EmitterState(Inactive Response=::NewEnumerator1, Life Cycle Mode=::NewEnumerator1, Loop Duration=2, UseLoopDelay=false)
   &SpawnRate(
    SpawnRate=&Lerp_Float(
     A=6
     B=3
     Alpha=&ClampFloat(Float=&Divide_Float(A=&VectorLength(Vector=$Engine.Owner.Velocity), B=500))
    )
    Use Spawn Probability=false
   )
  ]
  particleSpawnScript [
   &InitializeParticle(
    Lifetime=&Lerp_Float(
     A=&RandomRangeFloat(Minimum=1.4, Maximum=2.2)
     B=0.8
     Alpha=&Divide_Float(A=&VectorLength(Vector=$Engine.Owner.Velocity), B=5000)
    )
    Mesh Scale Max=<1 1.5 0.6>
    Mesh Scale Min=<0.2 0.2 0.2>
    Mesh Scale Mode=::NewEnumerator2
    UsePositionOffset=false
    Write Material Random=true
   )
   &InitialMeshOrientation(Use Rotation Vector=false)
   &Color(Color=&MakeLinearColorFromVectorAndFloat("Vector (RGB)"=<1 1 1>, "Float (Alpha)"=&ClampFloat(Float=1))) {enabled=false}
   &DynamicMaterialParameters(
    Index 0 Param 1=1
    Index 0 Param 2=10
    Index 0 Param 3=3
    Index 0 Param 4=&RandomRangeFloat(Minimum=50, Maximum=55, Evaluation Type=::NewEnumerator1)
    Module.Param0WriteEnabled=true
    Module.Param1WriteEnabled=true
    Module.Param2WriteEnabled=true
    Module.Param3WriteEnabled=false
   ) {module=DynamicMaterialParameters002}
   &SkeletalMeshLocation(
    MaskSpawnGroup=false
    Mesh Sampling Type=::NewEnumerator5
    OverrideRandomSeed=false
    Skeletal Mesh=DataInterface() {propertyValues {
      DefaultMesh=null
      ExcludeBoneName=root
      OptionalComponentClass=null
      PreviewMesh=/BRCosmetics/Characters/Player/Male/Medium/Bodies/M_MED_Psyche/Meshes/M_MED_Psyche
      SkinningMode=SkinOnTheFly
      SoftSourceActor=null
      SourceMode=Default
      UvSetIndex=0
      WholeMeshLOD=-1
      bExcludeBone=false
      bReadDeformedGeometry=true
      bRequireCurrentFrameData=true
      ComponentTags []
      FilteredBones []
      FilteredSockets [FX_Goo_01,FX_Goo_02,FX_Goo_03,FX_Goo_04,FX_Goo_05,FX_Goo_06,FX_Goo_07,FX_Goo_08,FX_Goo_09,FX_Goo_10,FX_Goo_11]
      MeshUserParameter {parameter {name=None, typeDefHandle {registeredTypeIndex=1}, varData []}}
      SamplingRegions []
     }}
   )
   &AddVelocity(
    Coordinate Space=::World
    Velocity=&Lerp_Vector(
     A=<0 0 10>
     B=<0 0 30>
     Alpha=&Divide_Float(A=&VectorLength(Vector=$Engine.Owner.Velocity), B=5000)
    )
   ) {module=AddVelocity001}
  ]
  particleUpdateScript [
   &ParticleState()
   &ScaleMeshSize(Scale Factor=&VectorFromFloat(Value=&FloatFromCurve(FloatCurve=DataInterface() {propertyValues {Curve {keys [
          {arriveTangent=9.1187, interpMode=RCIM_Cubic, time=0, value=0}
          {arriveTangent=0.3759, interpMode=RCIM_Cubic, leaveTangent=0.3759, time=0.2487, value=1.1246}
          {arriveTangent=-1.497, interpMode=RCIM_Cubic, leaveTangent=-1.497, time=0.7835, value=0.2945}
          <at 1 0 RCIM_Cubic>
         ]}}}))) {module=ScaleMeshSize001}
   &ScaleColor(
    Scale Alpha=&FloatFromCurve(FloatCurve=DataInterface() {propertyValues {Curve {keys [<at 0.5 1.0113 RCIM_Cubic>,<at 1.0011 0.002 RCIM_Cubic>]}}})
    ScaleA=true
    ScaleRGB=false
   ) {enabled=false}
   &DynamicMaterialParameters(
    Index 0 Param 1=5
    Index 0 Param 2=12
    Index 0 Param 3=&FloatFromCurve(FloatCurve=DataInterface() {propertyValues {Curve {keys [<at 0 5 RCIM_Cubic>, {arriveTangent=-4.1151, interpMode=RCIM_Cubic, leaveTangent=-4.1151, tangentMode=RCTM_User, time=1, value=1}]}}})
    Module.Param0WriteEnabled=true
    Module.Param1WriteEnabled=true
    Module.Param2WriteEnabled=true
    Module.Param3WriteEnabled=false
   )
   &AccelerationForce(Acceleration=&Multiply_VectorByFloat(
     Vector=&RandomRangeVector(Minimum=<-3 -3 5>, Maximum=<3 3 10>, Evaluation Type=::NewEnumerator1)
     Float=1
    ))
   &Drag(Drag=1.5, Use Linear Drag=true, Use Rotational Drag=false) {enabled=false}
   &SolveForcesAndVelocity(Clamp Velocity=false, Limit Acceleration=false)
   &MeshRotationForce(
    Coordinate Space:ENiagaraMeshTransforms=::NewEnumerator3
    Rotation=&RandomRangeVector(Minimum=<-4 -4 -4>, Maximum=<4 4 4>)
   )
   &SolveRotationalForcesAndVelocity()
  ]
  rendererClasses [&NiagaraMeshRendererProperties]
  renderers [{rendererClass=&NiagaraMeshRendererProperties, rendererIndex=0}]
 }}
summary {
 systemName=NS_Psyche_Goop_Idle
 emitters [{emitterName=Sludge_Base001, rendererClasses [&NiagaraMeshRendererProperties]}]
 userVariables [
  User.bIsFrontEnd {type=NiagaraBool}
  User.bIsFrontEndPreview {type=NiagaraBool}
  User.SkeletalMesh {
   type=NiagaraDataInterfaceSkeletalMesh
   defaultValue {
    struct=NiagaraExt_VariableValue_DataInterface
    value {dataInterface=&self:NiagaraDataInterfaceSkeletalMesh_0}
   }
  }
  User.SlurpColorAlpha {type=NiagaraFloat, defaultValue=0, description=""}
 ]
}
