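# Control Rig for the M_MED_Psyche body that drives corrective deform bones at the elbows, knees, shoulders, pectorals and thighs.
# Joint Euler angles are remapped to clamped weights that blend each bone toward hand-set local transforms with rotations offset from initial pose.
# Knees and elbows layer two angle ranges; shoulders and hips use one absolute angle.
# Only the elbow chain links to Begin Execution; the other chain, ending in debug drawing, appears unconnected.
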ControlRigBlueprint /BRCosmetics/Characters/Player/Male/Medium/Bodies/M_MED_Psyche/Meshes/M_MED_Psyche_Deform_CtrlRig_sculpters
graphs [
 RigVMFunctionLibrary {links [], nodes []}
 RigVMModel {nodes [
   n1 = &RigVMFunction_MathQuaternionToEuler(RotationOrder=ZYX, Value=&RigUnit_GetRelativeBoneTransform(Bone=calf_l, Space=thigh_l).Transform.Rotation)
   n2 = &RigVMFunction_MathFloatRemap(SourceMaximum=114, SourceMinimum=4, TargetMinimum=0, bClamp=true, Value=@n1.Result.Z)
   n3 = &RigVMFunction_MathQuaternionToEuler(
    RotationOrder=ZYX
    Value=/Script/ControlRig.RigUnit_GetInitialBoneTransform(Bone=deform_knee_bk_01_l, Space=LocalSpace) {is=struct RigVMUnitNode}.Transform.Rotation
   )
   /Script/ControlRig.RigUnit_ModifyBoneTransforms(
    BoneToModify=((Bone="deform_knee_bk_01_l",Transform=(Rotation=(X=0,Y=0,Z=0,W=1),Translation=(X=-4.319,Y=-4.798,Z=0),Scale3D=(X=1,Y=1,Z=0.7))))
    ExecutePin=()
    Mode=OverrideLocal
    WeightMaximum=1
    Weight=@n2.Result
    BoneToModify.0.Transform.Rotation=&RigVMFunction_MathQuaternionFromEuler(Euler=/Script/RigVM.RigVMFunction_MathVectorAdd(B=(X=0,Y=0,Z=0), A=@n3.Result) {is=struct RigVMUnitNode, template="Add::Execute(in A,in B,out Result)"}.Result).Result
   ) {is=struct RigVMUnitNode}
   n4 = &RigVMFunction_MathQuaternionToEuler(RotationOrder=ZYX, Value=&RigUnit_GetInitialBoneTransform(Bone=deform_knee_fr_01_l, Space=LocalSpace).Transform.Rotation)
   > &RigUnit_ModifyBoneTransforms(
    BoneToModify=((Bone="deform_knee_fr_01_l",Transform=(Rotation=(X=0,Y=0,Z=0,W=1),Translation=(X=0.258,Y=-2,Z=0.009),Scale3D=(X=1,Y=1,Z=1))))
    Mode=OverrideLocal
    WeightMaximum=1
    WeightMinimum=0
    Weight=@n2.Result
    BoneToModify.0.Transform.Rotation=&RigVMFunction_MathQuaternionFromEuler(Euler=&RigVMFunction_MathVectorAdd(B=(X=0,Y=0,Z=-15), A=@n4.Result) {template="Add::Execute(in A,in B,out Result)"}.Result).Result
   )
   n5 = &RigVMFunction_MathFloatRemap(SourceMaximum=130, SourceMinimum=113, TargetMinimum=0, bClamp=true, Value=@n1.Result.Z)
   > &RigUnit_ModifyBoneTransforms(
    BoneToModify=((Bone="deform_knee_bk_01_l",Transform=(Rotation=(X=0,Y=0,Z=0,W=1),Translation=(X=-10.633,Y=-3.561,Z=0),Scale3D=(X=1,Y=1,Z=1))))
    Mode=OverrideLocal
    WeightMaximum=1
    WeightMinimum=0
    Weight=@n5.Result
    BoneToModify.0.Transform.Rotation=&RigVMFunction_MathQuaternionFromEuler(Euler=&RigVMFunction_MathVectorAdd(B=(X=0,Y=0,Z=0), A=@n3.Result) {template="Add::Execute(in A,in B,out Result)"}.Result).Result
   )
   > &RigUnit_ModifyBoneTransforms(
    BoneToModify=((Bone="deform_knee_fr_01_l",Transform=(Rotation=(X=0,Y=0,Z=0,W=1),Translation=(X=0.258,Y=-3.591,Z=0.009),Scale3D=(X=1,Y=1,Z=1))))
    Mode=OverrideLocal
    WeightMaximum=1
    WeightMinimum=0
    Weight=@n5.Result
    BoneToModify.0.Transform.Rotation=&RigVMFunction_MathQuaternionFromEuler(Euler=&RigVMFunction_MathVectorAdd(B=(X=0,Y=0,Z=-51.2), A=@n4.Result) {template="Add::Execute(in A,in B,out Result)"}.Result).Result
   )
   n6 = &RigVMFunction_MathQuaternionToEuler(RotationOrder=ZYX, Value=&RigUnit_GetRelativeBoneTransform(Bone=calf_r, Space=thigh_r).Transform.Rotation)
   n7 = &RigVMFunction_MathFloatRemap(SourceMaximum=90, SourceMinimum=4, TargetMinimum=0, bClamp=true, Value=@n6.Result.Z)
   n8 = &RigVMFunction_MathQuaternionToEuler(RotationOrder=ZYX, Value=&RigUnit_GetInitialBoneTransform(Bone=deform_knee_bk_01_r, Space=LocalSpace).Transform.Rotation)
   > &RigUnit_ModifyBoneTransforms(
    BoneToModify=((Bone="deform_knee_bk_01_r",Transform=(Rotation=(X=0,Y=0,Z=0,W=1),Translation=(X=4.319,Y=4.798,Z=0),Scale3D=(X=1,Y=2.022,Z=1.95))))
    Mode=OverrideLocal
    WeightMaximum=1
    Weight=@n7.Result
    BoneToModify.0.Transform.Rotation=&RigVMFunction_MathQuaternionFromEuler(Euler=&RigVMFunction_MathVectorAdd(B=(X=0,Y=0,Z=0), A=@n8.Result) {template="Add::Execute(in A,in B,out Result)"}.Result).Result
   )
   n9 = &RigVMFunction_MathQuaternionToEuler(RotationOrder=ZYX, Value=&RigUnit_GetInitialBoneTransform(Bone=deform_knee_fr_01_r, Space=LocalSpace).Transform.Rotation)
   > &RigUnit_ModifyBoneTransforms(
    BoneToModify=((Bone="deform_knee_fr_01_r",Transform=(Rotation=(X=0,Y=0,Z=0,W=1),Translation=(X=-1.701,Y=-1.197,Z=0),Scale3D=(X=1,Y=1,Z=1))))
    Mode=OverrideLocal
    WeightMaximum=1
    WeightMinimum=0
    Weight=@n7.Result
    BoneToModify.0.Transform.Rotation=&RigVMFunction_MathQuaternionFromEuler(Euler=&RigVMFunction_MathVectorAdd(B=(X=0,Y=0,Z=0), A=@n9.Result) {template="Add::Execute(in A,in B,out Result)"}.Result).Result
   )
   n10 = &RigVMFunction_MathFloatRemap(SourceMaximum=130, SourceMinimum=89, TargetMinimum=0, bClamp=true, Value=@n6.Result.Z)
   > &RigUnit_ModifyBoneTransforms(
    BoneToModify=((Bone="deform_knee_bk_01_r",Transform=(Rotation=(X=0,Y=0,Z=0,W=1),Translation=(X=10.633,Y=3.561,Z=0),Scale3D=(X=1,Y=1,Z=1.777))))
    Mode=OverrideLocal
    WeightMaximum=1
    WeightMinimum=0
    Weight=@n10.Result
    BoneToModify.0.Transform.Rotation=&RigVMFunction_MathQuaternionFromEuler(Euler=&RigVMFunction_MathVectorAdd(B=(X=0,Y=0,Z=0), A=@n8.Result) {template="Add::Execute(in A,in B,out Result)"}.Result).Result
   )
   > &RigUnit_ModifyBoneTransforms(
    BoneToModify=((Bone="deform_knee_fr_01_r",Transform=(Rotation=(X=0,Y=0,Z=0,W=1),Translation=(X=-1.791,Y=-1.35,Z=0),Scale3D=(X=1.913,Y=1,Z=1))))
    Mode=OverrideLocal
    WeightMaximum=1
    WeightMinimum=0
    Weight=@n10.Result
    BoneToModify.0.Transform.Rotation=&RigVMFunction_MathQuaternionFromEuler(Euler=&RigVMFunction_MathVectorAdd(B=(X=0,Y=0,Z=0), A=@n9.Result) {template="Add::Execute(in A,in B,out Result)"}.Result).Result
   )
   n11 = &RigUnit_GetRelativeBoneTransform(Bone=upperarm_l, Space=spine_05)
   n12 = &RigVMFunction_MathFloatRemap(
    SourceMaximum=50
    SourceMinimum=170
    TargetMinimum=0
    bClamp=true
    Value=/Script/RigVM.RigVMFunction_MathFloatAbs(Value=&RigVMFunction_MathQuaternionToEuler(RotationOrder=ZYX, Value=@n11.Transform.Rotation).Result.Z) {is=struct RigVMUnitNode, template="Absolute::Execute(in Value,out Result)"}.Result
   )
   n13 = &RigVMFunction_MathQuaternionToEuler(RotationOrder=ZYX, Value=&RigUnit_GetInitialBoneTransform(Bone=deform_pectoral_fr_l, Space=LocalSpace).Transform.Rotation)
   > &RigUnit_ModifyBoneTransforms(
    BoneToModify=((Bone="deform_pectoral_fr_l",Transform=(Rotation=(X=0,Y=0,Z=0,W=1),Translation=(X=6.262,Y=5.9,Z=4.98),Scale3D=(X=1,Y=1,Z=1))))
    Mode=OverrideLocal
    WeightMaximum=1
    Weight=@n12.Result
    BoneToModify.0.Transform.Rotation=&RigVMFunction_MathQuaternionFromEuler(Euler=&RigVMFunction_MathVectorAdd(B=(X=0,Y=0,Z=0), A=@n13.Result) {template="Add::Execute(in A,in B,out Result)"}.Result).Result
   )
   n14 = &RigVMFunction_MathQuaternionToEuler(RotationOrder=ZYX, Value=&RigUnit_GetInitialBoneTransform(Bone=deform_deltoid_fr_l, Space=LocalSpace).Transform.Rotation)
   > &RigUnit_ModifyBoneTransforms(
    BoneToModify=((Bone="deform_deltoid_fr_l",Transform=(Rotation=(X=0,Y=0,Z=0,W=1),Translation=(X=5.008,Y=3.542,Z=0),Scale3D=(X=1,Y=1,Z=1))))
    Mode=OverrideLocal
    WeightMaximum=1
    WeightMinimum=0
    Weight=@n12.Result
    BoneToModify.0.Transform.Rotation=&RigVMFunction_MathQuaternionFromEuler(Euler=&RigVMFunction_MathVectorAdd(B=(X=0,Y=0,Z=-51), A=@n14.Result) {template="Add::Execute(in A,in B,out Result)"}.Result).Result
   )
   n15 = &RigUnit_GetRelativeBoneTransform(Bone=upperarm_r, Space=spine_05)
   n16 = &RigVMFunction_MathFloatRemap(
    SourceMaximum=123
    SourceMinimum=30
    TargetMinimum=0
    bClamp=true
    Value=&RigVMFunction_MathFloatAbs(Value=&RigVMFunction_MathQuaternionToEuler(RotationOrder=ZYX, Value=@n15.Transform.Rotation).Result.Z) {template="Absolute::Execute(in Value,out Result)"}.Result
   )
   RigUnit_GetInitialBoneTransform = &RigUnit_GetInitialBoneTransform(Bone=deform_pectoral_fr_r, Space=LocalSpace)
   n17 = &RigVMFunction_MathQuaternionToEuler(RotationOrder=ZYX, Value=@RigUnit_GetInitialBoneTransform.Transform.Rotation)
   > &RigUnit_ModifyBoneTransforms(
    BoneToModify=((Bone="deform_pectoral_fr_r",Transform=(Rotation=(X=0,Y=0,Z=0,W=1),Translation=(X=-6.262,Y=-5.9,Z=-4.98),Scale3D=(X=1,Y=1,Z=1))))
    Mode=OverrideLocal
    WeightMaximum=1
    Weight=@n16.Result
    BoneToModify.0.Transform.Rotation=&RigVMFunction_MathQuaternionFromEuler(Euler=&RigVMFunction_MathVectorAdd(B=(X=0,Y=0,Z=0), A=@n17.Result) {template="Add::Execute(in A,in B,out Result)"}.Result).Result
   )
   n18 = &RigVMFunction_MathQuaternionToEuler(RotationOrder=ZYX, Value=&RigUnit_GetInitialBoneTransform(Bone=deform_deltoid_fr_r, Space=LocalSpace).Transform.Rotation)
   > &RigUnit_ModifyBoneTransforms(
    BoneToModify=((Bone="deform_deltoid_fr_r",Transform=(Rotation=(X=0,Y=0,Z=0,W=1),Translation=(X=-5.008,Y=-3.542,Z=0),Scale3D=(X=1,Y=1,Z=1))))
    Mode=OverrideLocal
    WeightMaximum=1
    WeightMinimum=0
    Weight=@n16.Result
    BoneToModify.0.Transform.Rotation=&RigVMFunction_MathQuaternionFromEuler(Euler=&RigVMFunction_MathVectorAdd(B=(X=0,Y=0,Z=-51), A=@n18.Result) {template="Add::Execute(in A,in B,out Result)"}.Result).Result
   )
   n19 = &RigUnit_GetRelativeBoneTransform(Bone=thigh_l, Space=pelvis)
   n20 = &RigVMFunction_MathFloatRemap(
    SourceMaximum=100
    SourceMinimum=2
    TargetMinimum=0
    bClamp=true
    Value=&RigVMFunction_MathFloatAbs(Value=&RigVMFunction_MathQuaternionToEuler(RotationOrder=ZYX, Value=@n19.Transform.Rotation).Result.Z) {template="Absolute::Execute(in Value,out Result)"}.Result
   )
   n21 = &RigVMFunction_MathQuaternionToEuler(RotationOrder=ZYX, Value=&RigUnit_GetInitialBoneTransform(Bone=deform_thigh_fr_l, Space=LocalSpace).Transform.Rotation)
   > &RigUnit_ModifyBoneTransforms(
    BoneToModify=((Bone="deform_thigh_fr_l",Transform=(Rotation=(X=0,Y=0,Z=0,W=1),Translation=(X=13.423,Y=8.962,Z=2.253),Scale3D=(X=1,Y=1,Z=1.175))))
    Mode=OverrideLocal
    WeightMaximum=1
    Weight=@n20.Result
    BoneToModify.0.Transform.Rotation=&RigVMFunction_MathQuaternionFromEuler(Euler=&RigVMFunction_MathVectorAdd(B=(X=0,Y=-14.949,Z=0), A=@n21.Result) {template="Add::Execute(in A,in B,out Result)"}.Result).Result
   )
   n22 = &RigVMFunction_MathQuaternionToEuler(RotationOrder=ZYX, Value=&RigUnit_GetInitialBoneTransform(Bone=deform_thigh_bk_l, Space=LocalSpace).Transform.Rotation)
   > &RigUnit_ModifyBoneTransforms(
    BoneToModify=((Bone="deform_thigh_bk_l",Transform=(Rotation=(X=0,Y=0,Z=0,W=1),Translation=(X=29.958,Y=-1.401,Z=0.029),Scale3D=(X=1,Y=1,Z=1.803))))
    Mode=OverrideLocal
    WeightMaximum=1
    WeightMinimum=0
    Weight=@n20.Result
    BoneToModify.0.Transform.Rotation=&RigVMFunction_MathQuaternionFromEuler(Euler=&RigVMFunction_MathVectorAdd(B=(X=0,Y=0,Z=0), A=@n22.Result) {template="Add::Execute(in A,in B,out Result)"}.Result).Result
   )
   n23 = &RigUnit_GetRelativeBoneTransform(Bone=thigh_r, Space=pelvis)
   n24 = &RigVMFunction_MathFloatRemap(
    SourceMaximum=77
    SourceMinimum=177
    TargetMinimum=0
    bClamp=true
    Value=&RigVMFunction_MathFloatAbs(Value=&RigVMFunction_MathQuaternionToEuler(RotationOrder=ZYX, Value=@n23.Transform.Rotation).Result.Z) {template="Absolute::Execute(in Value,out Result)"}.Result
   )
   n25 = &RigVMFunction_MathQuaternionToEuler(RotationOrder=ZYX, Value=&RigUnit_GetInitialBoneTransform(Bone=deform_thigh_fr_r, Space=LocalSpace).Transform.Rotation)
   > &RigUnit_ModifyBoneTransforms(
    BoneToModify=((Bone="deform_thigh_fr_r",Transform=(Rotation=(X=0,Y=0,Z=0,W=1),Translation=(X=-13.423,Y=-8.962,Z=-2.253),Scale3D=(X=1,Y=1,Z=1.175))))
    Mode=OverrideLocal
    WeightMaximum=1
    Weight=@n24.Result
    BoneToModify.0.Transform.Rotation=&RigVMFunction_MathQuaternionFromEuler(Euler=&RigVMFunction_MathVectorAdd(B=(X=0,Y=-14.949,Z=0), A=@n25.Result) {template="Add::Execute(in A,in B,out Result)"}.Result).Result
   )
   n26 = &RigVMFunction_MathQuaternionToEuler(RotationOrder=ZYX, Value=&RigUnit_GetInitialBoneTransform(Bone=deform_thigh_bk_r, Space=LocalSpace).Transform.Rotation)
   > &RigUnit_ModifyBoneTransforms(
    BoneToModify=((Bone="deform_thigh_bk_r",Transform=(Rotation=(X=0,Y=0,Z=0,W=1),Translation=(X=-29.958,Y=1.401,Z=-0.029),Scale3D=(X=1,Y=1,Z=1.803))))
    Mode=OverrideLocal
    WeightMaximum=1
    WeightMinimum=0
    Weight=@n24.Result
    BoneToModify.0.Transform.Rotation=&RigVMFunction_MathQuaternionFromEuler(Euler=&RigVMFunction_MathVectorAdd(B=(X=0,Y=0,Z=0), A=@n26.Result) {template="Add::Execute(in A,in B,out Result)"}.Result).Result
   )
   > /Script/ControlRig.RigUnit_DebugHierarchy(
    Color=(R=1,G=1,B=1,A=1)
    DebugDrawSettings=(DepthPriority=SDPG_Foreground,Lifetime=-1)
    Items=()
    Scale=3
    Thickness=0.1
    WorldOffset=(Rotation=(X=0,Y=0,Z=0,W=1),Translation=(X=0,Y=0,Z=0),Scale3D=(X=1,Y=1,Z=1))
    bEnabled=True
   ) {is=struct RigVMUnitNode}
   RigUnit_BeginExecution = &RigUnit_BeginExecution()
   RigUnit_GetRelativeBoneTransform = &RigUnit_GetRelativeBoneTransform(Bone=lowerarm_l, Space=upperarm_l)
   n27 = &RigVMFunction_MathQuaternionToEuler(RotationOrder=ZYX, Value=@RigUnit_GetRelativeBoneTransform.Transform.Rotation)
   n28 = &RigVMFunction_MathFloatRemap(SourceMaximum=92, SourceMinimum=28, TargetMinimum=0, bClamp=true, Value=@n27.Result.Z)
   n29 = &RigVMFunction_MathQuaternionToEuler(RotationOrder=ZYX, Value=&RigUnit_GetInitialBoneTransform(Bone=deform_elbow_bk_l, Space=LocalSpace).Transform.Rotation)
   &RigUnit_ModifyBoneTransforms(
    BoneToModify=((Bone="deform_elbow_bk_l",Transform=(Rotation=(X=0,Y=0,Z=0,W=1),Translation=(X=-2.858,Y=-2.932,Z=-0.666),Scale3D=(X=1,Y=1,Z=1))))
    Mode=OverrideLocal
    WeightMaximum=1
    Weight=@n28.Result
    BoneToModify.0.Transform.Rotation=&RigVMFunction_MathQuaternionFromEuler(Euler=&RigVMFunction_MathVectorAdd(B=(X=0,Y=0,Z=0), A=@n29.Result) {template="Add::Execute(in A,in B,out Result)"}.Result).Result
    ExecutePin=@RigUnit_BeginExecution.ExecutePin
   )
   n30 = &RigVMFunction_MathQuaternionToEuler(RotationOrder=ZYX, Value=&RigUnit_GetInitialBoneTransform(Bone=deform_elbow_fr_l, Space=LocalSpace).Transform.Rotation)
   > &RigUnit_ModifyBoneTransforms(
    BoneToModify=((Bone="deform_elbow_fr_l",Transform=(Rotation=(X=0,Y=0,Z=0,W=1),Translation=(X=24.635,Y=2.261,Z=-0.637),Scale3D=(X=1,Y=1,Z=1.125))))
    Mode=OverrideLocal
    WeightMaximum=1
    WeightMinimum=0
    Weight=@n28.Result
    BoneToModify.0.Transform.Rotation=&RigVMFunction_MathQuaternionFromEuler(Euler=&RigVMFunction_MathVectorAdd(B=(X=0,Y=0,Z=0), A=@n30.Result) {template="Add::Execute(in A,in B,out Result)"}.Result).Result
   )
   n31 = &RigVMFunction_MathFloatRemap(SourceMaximum=148, SourceMinimum=91, TargetMinimum=0, bClamp=true, Value=@n27.Result.Z)
   > &RigUnit_ModifyBoneTransforms(
    BoneToModify=((Bone="deform_elbow_bk_l",Transform=(Rotation=(X=0,Y=0,Z=0,W=1),Translation=(X=-3.59,Y=-2.934,Z=-0.716),Scale3D=(X=1,Y=1,Z=1))))
    Mode=OverrideLocal
    WeightMaximum=1
    Weight=@n31.Result
    BoneToModify.0.Transform.Rotation=&RigVMFunction_MathQuaternionFromEuler(Euler=&RigVMFunction_MathVectorAdd(B=(X=0,Y=0,Z=-13.6), A=@n29.Result) {template="Add::Execute(in A,in B,out Result)"}.Result).Result
   )
   > &RigUnit_ModifyBoneTransforms(
    BoneToModify=((Bone="deform_elbow_fr_l",Transform=(Rotation=(X=0,Y=0,Z=0,W=1),Translation=(X=15.651,Y=1.255,Z=-0.922),Scale3D=(X=1,Y=1,Z=1.337))))
    Mode=OverrideLocal
    WeightMaximum=1
    WeightMinimum=0
    Weight=@n31.Result
    BoneToModify.0.Transform.Rotation=&RigVMFunction_MathQuaternionFromEuler(Euler=&RigVMFunction_MathVectorAdd(B=(X=0,Y=0,Z=28), A=@n30.Result) {template="Add::Execute(in A,in B,out Result)"}.Result).Result
   )
   n32 = &RigVMFunction_MathQuaternionToEuler(RotationOrder=ZYX, Value=&RigUnit_GetRelativeBoneTransform(Bone=lowerarm_r, Space=upperarm_r).Transform.Rotation)
   n33 = &RigVMFunction_MathFloatRemap(SourceMaximum=92, SourceMinimum=28, TargetMinimum=0, bClamp=true, Value=@n32.Result.Z)
   n34 = &RigVMFunction_MathQuaternionToEuler(RotationOrder=ZYX, Value=&RigUnit_GetInitialBoneTransform(Bone=deform_elbow_bk_r, Space=LocalSpace).Transform.Rotation)
   > &RigUnit_ModifyBoneTransforms(
    BoneToModify=((Bone="deform_elbow_bk_r",Transform=(Rotation=(X=0,Y=0,Z=0,W=1),Translation=(X=2.858,Y=2.932,Z=0.666),Scale3D=(X=1,Y=1,Z=1))))
    Mode=OverrideLocal
    WeightMaximum=1
    Weight=@n33.Result
    BoneToModify.0.Transform.Rotation=&RigVMFunction_MathQuaternionFromEuler(Euler=&RigVMFunction_MathVectorAdd(B=(X=0,Y=0,Z=0), A=@n34.Result) {template="Add::Execute(in A,in B,out Result)"}.Result).Result
   )
   n35 = &RigVMFunction_MathQuaternionToEuler(RotationOrder=ZYX, Value=&RigUnit_GetInitialBoneTransform(Bone=deform_elbow_fr_r, Space=LocalSpace).Transform.Rotation)
   > &RigUnit_ModifyBoneTransforms(
    BoneToModify=((Bone="deform_elbow_fr_r",Transform=(Rotation=(X=0,Y=0,Z=0,W=1),Translation=(X=-24.635,Y=-2.261,Z=0.637),Scale3D=(X=1,Y=1,Z=1.125))))
    Mode=OverrideLocal
    WeightMaximum=1
    WeightMinimum=0
    Weight=@n33.Result
    BoneToModify.0.Transform.Rotation=&RigVMFunction_MathQuaternionFromEuler(Euler=&RigVMFunction_MathVectorAdd(B=(X=0,Y=0,Z=0), A=@n35.Result) {template="Add::Execute(in A,in B,out Result)"}.Result).Result
   )
   n36 = &RigVMFunction_MathFloatRemap(SourceMaximum=148, SourceMinimum=91, TargetMinimum=0, bClamp=true, Value=@n32.Result.Z)
   > &RigUnit_ModifyBoneTransforms(
    BoneToModify=((Bone="deform_elbow_bk_r",Transform=(Rotation=(X=0,Y=0,Z=0,W=1),Translation=(X=3.59,Y=2.934,Z=0.716),Scale3D=(X=1,Y=1,Z=1))))
    Mode=OverrideLocal
    WeightMaximum=1
    Weight=@n36.Result
    BoneToModify.0.Transform.Rotation=&RigVMFunction_MathQuaternionFromEuler(Euler=&RigVMFunction_MathVectorAdd(B=(X=0,Y=0,Z=-13.6), A=@n34.Result) {template="Add::Execute(in A,in B,out Result)"}.Result).Result
   )
   > &RigUnit_ModifyBoneTransforms(
    BoneToModify=((Bone="deform_elbow_fr_r",Transform=(Rotation=(X=0,Y=0,Z=0,W=1),Translation=(X=-15.651,Y=-1.255,Z=0.922),Scale3D=(X=1,Y=1,Z=1.337))))
    Mode=OverrideLocal
    WeightMaximum=1
    WeightMinimum=0
    Weight=@n36.Result
    BoneToModify.0.Transform.Rotation=&RigVMFunction_MathQuaternionFromEuler(Euler=&RigVMFunction_MathVectorAdd(B=(X=0,Y=0,Z=28), A=@n35.Result) {template="Add::Execute(in A,in B,out Result)"}.Result).Result
   )
   RigVMVariableNode() {variable=KneeRot_fr_l_114deg}
   RigVMVariableNode() {variable=KneeTrans_fr_l_114deg}
   RigVMVariableNode() {variable=KneeScale_fr_l_114deg}
   RigVMVariableNode() {variable=KneeScale_fr_l_130deg}
   RigVMVariableNode() {variable=KneeRot_fr_l_130deg}
   RigVMVariableNode() {variable=KneeTrans_fr_l_130deg}
  ]}
]
hierarchy [
 Root {loc=<0 0 0>, rot=<0 0 0 1>}
 pelvis {parent=Root, loc=<0 -0.7071 93.1191>, rot=<0.0127 0.707 -0.0127 -0.707>}
 Spine_01 {parent=pelvis, loc=<6.7494 -0.6388 0>, rot=<0 0 0.034 -0.9994>}
 Spine_02 {parent=Spine_01, loc=<7.2518 -0.035 0>, rot=<0 0 0.0238 -0.9997>}
 Spine_03 {parent=Spine_02, loc=<6.0104 0 0>, rot=<0 0 0.0149 -0.9999>}
 spine_04 {parent=Spine_03, loc=<8.1024 0 0>, rot=<0 0 -0.0132 -0.9999>}
 spine_05 {parent=spine_04, loc=<10.0146 0 0>, rot=<0 0 -0.0015 -1>}
 clavicle_l {parent=spine_05, loc=<4.3575 -0.1466 -2.3223>, rot=<0.0465 -0.7289 -0.009 -0.6829>}
 upperarm_l {parent=clavicle_l, loc=<15.5111 0 0>, rot=<0.0226 -0.3234 -0.0088 -0.9459>}
 lowerarm_l {parent=upperarm_l, loc=<26.6514 0 0>, rot=<0.0391 -0.0099 -0.2453 -0.9686>}
 hand_l {parent=lowerarm_l, loc=<22.3491 0 0>, rot=<-0.7081 -0.0507 0.0975 -0.6975>}
 index_metacarpal_l {parent=hand_l, loc=<1.6535 0.0183 -2.96>, rot=<0.0158 -0.0124 -0.0173 -0.9996>}
 index_01_l {parent=index_metacarpal_l, loc=<8.3008 0.0357 -0.0524>, rot=<-0.0044 0.0485 0.0924 -0.9945>}
 index_02_l {parent=index_01_l, loc=<3.3587 -0.3004 0.0452>, rot=<-0.002 0.0027 0.3225 -0.9466>}
 index_03_l {parent=index_02_l, loc=<2.5578 -0.123 -0.0252>, rot=<-0.002 0.0067 0.2558 -0.9667>}
 ring_metacarpal_l {parent=hand_l, loc=<2.2074 -0.3625 1.3638>, rot=<-0.0284 0.0722 -0.0458 -0.9959>}
 ring_01_l {parent=ring_metacarpal_l, loc=<7.0295 -0.6197 -0.1446>, rot=<-0.0489 -0.0252 0.1892 -0.9804>}
 ring_02_l {parent=ring_01_l, loc=<3.6174 0.4647 -0.1027>, rot=<0 0 0.2806 -0.9598>}
 ring_03_l {parent=ring_02_l, loc=<2.4152 0.114 0.0199>, rot=<0 0 0.1189 -0.9929>}
 thumb_01_l {parent=hand_l, loc=<1.3118 -0.3191 -2.459>, rot=<0.6546 -0.4042 0.0664 -0.6354>}
 thumb_02_l {parent=thumb_01_l, loc=<5.9336 -0.0448 0.0401>, rot=<-0.0037 0.0213 0.298 -0.9543>}
 thumb_03_l {parent=thumb_02_l, loc=<2.2223 0.3682 -0.0907>, rot=<0.0118 -0.013 0.0066 -0.9998>}
 lowerarm_twist_01_l {parent=lowerarm_l, loc=<17.8549 0 0>, rot=<0 0 0 -1>}
 lowerarm_twist_02_l {parent=lowerarm_l, loc=<9.6199 0 0>, rot=<0 0 0 -1>}
 deform_elbow_bk_l {parent=lowerarm_l, loc=<-1.2853 -2.9274 -0.5607>, rot=<-0.0285 0.0335 -0.0024 -0.999>}
 upperarm_twist_01_l {parent=upperarm_l, loc=<0.0041 0 0>, rot=<0 0 0 -1>}
 upperarm_twist_02_l {parent=upperarm_l, loc=<13.7941 0 0>, rot=<0 0 0 -1>}
 deform_elbow_fr_l {parent=upperarm_l, loc=<26.5048 0.1875 -0.5816>, rot=<0.0021 0.0156 -0.185 -0.9826>}
 clavicle_r {parent=spine_05, loc=<4.3575 -0.1466 2.3223>, rot=<0.7289 0.0465 -0.6829 0.009>}
 upperarm_r {parent=clavicle_r, loc=<-15.5111 0 0>, rot=<0.0226 -0.3234 -0.0088 -0.9459>}
 lowerarm_r {parent=upperarm_r, loc=<-26.6515 0 0>, rot=<0.0391 -0.0099 -0.2453 -0.9686>}
 hand_r {parent=lowerarm_r, loc=<-22.3492 0 0>, rot=<-0.7081 -0.0507 0.0975 -0.6975>}
 index_metacarpal_r {parent=hand_r, loc=<-1.6534 -0.0181 2.9601>, rot=<0.0158 -0.0124 -0.0173 -0.9996>}
 index_01_r {parent=index_metacarpal_r, loc=<-8.3008 -0.0357 0.0524>, rot=<-0.0044 0.0485 0.0924 -0.9945>}
 index_02_r {parent=index_01_r, loc=<-3.3589 0.3003 -0.0453>, rot=<-0.002 0.0027 0.3225 -0.9466>}
 index_03_r {parent=index_02_r, loc=<-2.5578 0.1231 0.0251>, rot=<-0.002 0.0067 0.2558 -0.9667>}
 ring_metacarpal_r {parent=hand_r, loc=<-2.2073 0.3627 -1.3638>, rot=<-0.0284 0.0722 -0.0458 -0.9959>}
 ring_01_r {parent=ring_metacarpal_r, loc=<-7.0297 0.6197 0.1446>, rot=<-0.0489 -0.0252 0.1892 -0.9804>}
 ring_02_r {parent=ring_01_r, loc=<-3.6171 -0.4648 0.1027>, rot=<0 0 0.2806 -0.9598>}
 ring_03_r {parent=ring_02_r, loc=<-2.4153 -0.1141 -0.0199>, rot=<0 0 0.1189 -0.9929>}
 thumb_01_r {parent=hand_r, loc=<-1.3119 0.3192 2.459>, rot=<0.6546 -0.4042 0.0664 -0.6354>}
 thumb_02_r {parent=thumb_01_r, loc=<-5.9336 0.0447 -0.0401>, rot=<-0.0037 0.0213 0.298 -0.9543>}
 thumb_03_r {parent=thumb_02_r, loc=<-2.2223 -0.3682 0.0907>, rot=<0.0118 -0.013 0.0066 -0.9998>}
 lowerarm_twist_01_r {parent=lowerarm_r, loc=<-17.855 0 0>, rot=<0 0 0 1>}
 lowerarm_twist_02_r {parent=lowerarm_r, loc=<-9.6199 0 0>, rot=<0 0 0 1>}
 deform_elbow_bk_r {parent=lowerarm_r, loc=<1.2852 2.9274 0.561>, rot=<-0.0285 0.0335 -0.0024 -0.999>}
 upperarm_twist_01_r {parent=upperarm_r, loc=<-0.0041 0 0>, rot=<0 0 0 1>}
 upperarm_twist_02_r {parent=upperarm_r, loc=<-13.7942 0 0>, rot=<0 0 0 1>}
 deform_elbow_fr_r {parent=upperarm_r, loc=<-26.5046 -0.1875 0.5813>, rot=<0.0021 0.0156 -0.185 -0.9826>}
 neck_01 {parent=spine_05, loc=<11.3138 0.1529 0>, rot=<0 0 -0.2003 -0.9797>}
 neck_02 {parent=neck_01, loc=<4.5627 0 0>, rot=<0 0 0.0504 -0.9987>}
 head {parent=neck_02, loc=<5.3914 0 0>, rot=<0 0 0.115 -0.9934>}
 head_offset {parent=neck_02, loc=<5.3914 0 0>, rot=<0 0 0.115 -0.9934>}
 thigh_l {parent=pelvis, loc=<-2.5693 -0.0109 -7.7501>, rot=<0.0012 0.0572 0.0155 -0.9982>}
 calf_l {parent=thigh_l, loc=<-39.0484 0 0>, rot=<0.0006 -0.0161 -0.0391 -0.9991>}
 foot_l {parent=calf_l, loc=<-42.561 0.1077 0.4292>, rot=<0.0418 -0.0455 0.0343 -0.9975>}
 ball_l {parent=foot_l, loc=<-6.1599 13.8264 0.6806>, rot=<0.0001 -0.0001 -0.7063 -0.7079>}
 calf_twist_01_l {parent=calf_l, loc=<-35.813 -0.0967 0.0397>, rot=<0.0237 0.0095 -0.0035 -0.9997>}
 calf_twist_02_l {parent=calf_l, loc=<-18.8064 -0.0967 0.0397>, rot=<0.0237 0.0095 -0.0035 -0.9997>}
 thigh_twist_01_l {parent=thigh_l, loc=<-21.6303 0 0>, rot=<0.0013 0 0 -1>}
 thigh_r {parent=pelvis, loc=<-2.5691 -0.0109 7.7501>, rot=<0.0572 -0.0012 0.9982 0.0155>}
 calf_r {parent=thigh_r, loc=<39.0487 0 0>, rot=<0.0006 -0.0161 -0.0391 -0.9991>}
 foot_r {parent=calf_r, loc=<42.5612 -0.1077 -0.4292>, rot=<0.0418 -0.0455 0.0343 -0.9975>}
 ball_r {parent=foot_r, loc=<6.1598 -13.8264 -0.6806>, rot=<0.0001 -0.0001 -0.7063 -0.7079>}
 calf_twist_01_r {parent=calf_r, loc=<35.8132 0.0967 -0.0397>, rot=<0.0237 0.0095 -0.0035 -0.9997>}
 calf_twist_02_r {parent=calf_r, loc=<18.8064 0.0967 -0.0397>, rot=<0.0237 0.0095 -0.0035 -0.9997>}
 thigh_twist_01_r {parent=thigh_r, loc=<21.6303 0 0>, rot=<0.0013 0 0 -1>}
 dyn_jacketSide_mstr_l {parent=pelvis, loc=<5.1426 -0.6401 -13.7152>, rot=<-0.0043 -0.9941 -0.0463 -0.0984>}
 dyn_jacketSide_01_l {parent=dyn_jacketSide_mstr_l, loc=<0 0 0>, rot=<0 0 0 -1>}
 dyn_jacketSide_02_l {parent=dyn_jacketSide_01_l, loc=<8.0946 0.0838 -0.0009>, rot=<0.0172 -0.0001 0.0157 -0.9997>}
 dyn_jacketSide_03_l {parent=dyn_jacketSide_02_l, loc=<14.3399 0.0096 -0.0046>, rot=<-0.0062 0.0009 0.0018 -1>}
 dyn_jacketSide_04_l {parent=dyn_jacketSide_03_l, loc=<14.34 -0.0174 0.0166>, rot=<0.0093 -0.0031 -0.0033 -0.9999>}
 dyn_jacketBack_mstr_l {parent=pelvis, loc=<5.2174 -10.7233 -6.2544>, rot=<0.0086 -0.7958 0.5994 -0.0861>}
 dyn_jacketBack_01_l {parent=dyn_jacketBack_mstr_l, loc=<0 0 0>, rot=<0 0 0 1>}
 dyn_jacketBack_02_l {parent=dyn_jacketBack_01_l, loc=<7.712 0 -0.0373>, rot=<0.0271 0.0057 0.0002 -0.9996>}
 dyn_jacketBack_03_l {parent=dyn_jacketBack_02_l, loc=<13.7342 0 -0.0632>, rot=<-0.0184 0.0097 -0.0002 -0.9998>}
 dyn_jacketBack_04_l {parent=dyn_jacketBack_03_l, loc=<13.7336 0 0.0018>, rot=<-0.0016 -0.0003 0 -1>}
 dyn_jacketFront_mstr_l {parent=pelvis, loc=<11.3152 7.6788 -8.1054>, rot=<-0.0087 -0.9613 -0.2573 -0.0978>}
 dyn_jacketFront_01_l {parent=dyn_jacketFront_mstr_l, loc=<0 0 0>, rot=<0 0 0 -1>}
 dyn_jacketFront_02_l {parent=dyn_jacketFront_01_l, loc=<15.1468 0.0584 -0.0006>, rot=<0.0172 -0.0001 0.0157 -0.9997>}
 dyn_jacketFront_03_l {parent=dyn_jacketFront_02_l, loc=<15.1459 0.0067 -0.0032>, rot=<-0.0062 0.0009 0.0018 -1>}
 dyn_jacketFront_04_l {parent=dyn_jacketFront_03_l, loc=<15.146 -0.0121 0.0116>, rot=<0.0093 -0.0031 -0.0033 -0.9999>}
 dyn_jacketSide_mstr_r {parent=pelvis, loc=<5.1426 -0.6401 13.7152>, rot=<0.0984 0.0463 0.9941 0.0043>}
 dyn_jacketSide_01_r {parent=dyn_jacketSide_mstr_r, loc=<0 0 0>, rot=<0 0 0 -1>}
 dyn_jacketSide_02_r {parent=dyn_jacketSide_01_r, loc=<8.0947 -0.0838 -0.0009>, rot=<-0.0172 -0.0001 -0.0157 -0.9997>}
 dyn_jacketSide_03_r {parent=dyn_jacketSide_02_r, loc=<14.3399 -0.0096 -0.0046>, rot=<0.0062 0.0009 -0.0018 -1>}
 dyn_jacketSide_04_r {parent=dyn_jacketSide_03_r, loc=<14.34 0.0174 0.0166>, rot=<-0.0093 -0.0031 0.0033 -0.9999>}
 dyn_jacketBack_mstr_r {parent=pelvis, loc=<5.2175 -10.7233 6.2544>, rot=<0.0861 -0.5994 0.7958 -0.0086>}
 dyn_jacketBack_01_r {parent=dyn_jacketBack_mstr_r, loc=<0 0 0>, rot=<0 0 0 1>}
 dyn_jacketBack_02_r {parent=dyn_jacketBack_01_r, loc=<7.712 0 -0.0373>, rot=<-0.0271 0.0057 -0.0002 -0.9996>}
 dyn_jacketBack_03_r {parent=dyn_jacketBack_02_r, loc=<13.7341 0 -0.0632>, rot=<0.0184 0.0097 0.0002 -0.9998>}
 dyn_jacketBack_04_r {parent=dyn_jacketBack_03_r, loc=<13.7336 0 0.0019>, rot=<0.0016 -0.0003 0 -1>}
 dyn_jacketFront_mstr_r {parent=pelvis, loc=<11.3152 7.6788 8.1054>, rot=<0.0978 0.2573 0.9613 0.0087>}
 dyn_jacketFront_01_r {parent=dyn_jacketFront_mstr_r, loc=<0 0 0>, rot=<0 0 0 1>}
 dyn_jacketFront_02_r {parent=dyn_jacketFront_01_r, loc=<15.1468 -0.0584 -0.0007>, rot=<-0.0172 -0.0001 -0.0157 -0.9997>}
 dyn_jacketFront_03_r {parent=dyn_jacketFront_02_r, loc=<15.146 -0.0067 -0.0032>, rot=<0.0062 0.0009 -0.0018 -1>}
 dyn_jacketFront_04_r {parent=dyn_jacketFront_03_r, loc=<15.146 0.0121 0.0115>, rot=<-0.0093 -0.0031 0.0033 -0.9999>}
 dyn_root_simspace {parent=Root, loc=<0 0 -1.6308>, rot=<0 0 0 1>}
 dyn_crouch_collision {parent=Root, loc=<0 0 0>, rot=<0 0 0 1>}
 r_knee_corrective {type=Curve, loc=<0 0 0>, rot=<0 0 0 1>}
 l_knee_corrective {type=Curve, loc=<0 0 0>, rot=<0 0 0 1>}
 Male_Med_Head_01_Pout {type=Curve, loc=<0 0 0>, rot=<0 0 0 1>}
 Male_Med_Head_02_Pout {type=Curve, loc=<0 0 0>, rot=<0 0 0 1>}
 Male_Med_Head_03_Pout {type=Curve, loc=<0 0 0>, rot=<0 0 0 1>}
 Male_Med_Head_05_Effort {type=Curve, loc=<0 0 0>, rot=<0 0 0 1>}
 Stretch {type=Curve, loc=<0 0 0>, rot=<0 0 0 1>}
 Squash {type=Curve, loc=<0 0 0>, rot=<0 0 0 1>}
 polySurface151 {type=Curve, loc=<0 0 0>, rot=<0 0 0 1>}
 l_cheek_puff {type=Curve, loc=<0 0 0>, rot=<0 0 0 1>}
 r_cheek_puff {type=Curve, loc=<0 0 0>, rot=<0 0 0 1>}
 l_smile_fix {type=Curve, loc=<0 0 0>, rot=<0 0 0 1>}
 r_smile_fix {type=Curve, loc=<0 0 0>, rot=<0 0 0 1>}
 jaw_open_fix {type=Curve, loc=<0 0 0>, rot=<0 0 0 1>}
 r_lip_sneer_fix {type=Curve, loc=<0 0 0>, rot=<0 0 0 1>}
 l_lip_sneer_fix {type=Curve, loc=<0 0 0>, rot=<0 0 0 1>}
 L_Blink {type=Curve, loc=<0 0 0>, rot=<0 0 0 1>}
 R_Blink {type=Curve, loc=<0 0 0>, rot=<0 0 0 1>}
 L_Arm_Stretch {type=Curve, loc=<0 0 0>, rot=<0 0 0 1>}
 Hair_Hat {type=Curve, loc=<0 0 0>, rot=<0 0 0 1>}
 Hair_Helmet {type=Curve, loc=<0 0 0>, rot=<0 0 0 1>}
 Male_Med_Head_02_What {type=Curve, loc=<0 0 0>, rot=<0 0 0 1>}
 Male_Med_Head_02_Smile {type=Curve, loc=<0 0 0>, rot=<0 0 0 1>}
 Male_Med_Head_02_Scowl {type=Curve, loc=<0 0 0>, rot=<0 0 0 1>}
 Male_Med_Head_02_Hmmph {type=Curve, loc=<0 0 0>, rot=<0 0 0 1>}
 Male_Med_Head_02_Grrr {type=Curve, loc=<0 0 0>, rot=<0 0 0 1>}
 Male_Med_Head_02_Effort {type=Curve, loc=<0 0 0>, rot=<0 0 0 1>}
 Male_Med_Head_03_What {type=Curve, loc=<0 0 0>, rot=<0 0 0 1>}
 Male_Med_Head_03_Grrr {type=Curve, loc=<0 0 0>, rot=<0 0 0 1>}
 Male_Med_Head_03_Effort {type=Curve, loc=<0 0 0>, rot=<0 0 0 1>}
 Male_Med_Head_03_Hmmph {type=Curve, loc=<0 0 0>, rot=<0 0 0 1>}
 Male_Med_Head_03_Scowl {type=Curve, loc=<0 0 0>, rot=<0 0 0 1>}
 Male_Med_Head_03_Smile {type=Curve, loc=<0 0 0>, rot=<0 0 0 1>}
 Male_Med_Head_05_Teeth_Low_What {type=Curve, loc=<0 0 0>, rot=<0 0 0 1>}
 Male_Med_Head_05_Teeth_Low_Pout {type=Curve, loc=<0 0 0>, rot=<0 0 0 1>}
 Male_Med_Head_05_Teeth_Low_Hmmph {type=Curve, loc=<0 0 0>, rot=<0 0 0 1>}
 Male_Med_Head_05_Teeth_Low_Grrr {type=Curve, loc=<0 0 0>, rot=<0 0 0 1>}
 Male_Med_Head_05_Teeth_Low_Effort {type=Curve, loc=<0 0 0>, rot=<0 0 0 1>}
 Male_Med_Head_05_What {type=Curve, loc=<0 0 0>, rot=<0 0 0 1>}
 Male_Med_Head_05_Grrr {type=Curve, loc=<0 0 0>, rot=<0 0 0 1>}
 Male_Med_Head_05_Hmmph {type=Curve, loc=<0 0 0>, rot=<0 0 0 1>}
 Male_Med_Head_05_Pout {type=Curve, loc=<0 0 0>, rot=<0 0 0 1>}
 Male_Med_Head_05_Tongue_What {type=Curve, loc=<0 0 0>, rot=<0 0 0 1>}
 Male_Med_Head_05_Tongue_Pout {type=Curve, loc=<0 0 0>, rot=<0 0 0 1>}
 Male_Med_Head_05_Tongue_Hmmph {type=Curve, loc=<0 0 0>, rot=<0 0 0 1>}
 Male_Med_Head_05_Tongue_Grrr {type=Curve, loc=<0 0 0>, rot=<0 0 0 1>}
 Male_Med_Head_05_Tongue_Effort {type=Curve, loc=<0 0 0>, rot=<0 0 0 1>}
 AdditiveLayer {type=Curve, loc=<0 0 0>, rot=<0 0 0 1>}
 RightHandIK {type=Curve, loc=<0 0 0>, rot=<0 0 0 1>}
 LeftHandIK {type=Curve, loc=<0 0 0>, rot=<0 0 0 1>}
 LeftHandFK {type=Curve, loc=<0 0 0>, rot=<0 0 0 1>}
 Male_Med_Head_01_Grrr {type=Curve, loc=<0 0 0>, rot=<0 0 0 1>}
 WeaponAiming {type=Curve, loc=<0 0 0>, rot=<0 0 0 1>}
 Male_Med_Head_01_Hmmph {type=Curve, loc=<0 0 0>, rot=<0 0 0 1>}
 DragonSlash {type=Curve, loc=<0 0 0>, rot=<0 0 0 1>}
 Male_Med_Head_01_Effort {type=Curve, loc=<0 0 0>, rot=<0 0 0 1>}
 LeftHankFK {type=Curve, loc=<0 0 0>, rot=<0 0 0 1>}
 blendOrient1 {type=Curve, loc=<0 0 0>, rot=<0 0 0 1>}
 blendParent1 {type=Curve, loc=<0 0 0>, rot=<0 0 0 1>}
 bottom_R_Corner_Down {type=Curve, loc=<0 0 0>, rot=<0 0 0 1>}
 bottom_R_Corner_Up {type=Curve, loc=<0 0 0>, rot=<0 0 0 1>}
 top_L_Corner_Up {type=Curve, loc=<0 0 0>, rot=<0 0 0 1>}
 top_L_Corner_Down {type=Curve, loc=<0 0 0>, rot=<0 0 0 1>}
 top_R_Corner_Down {type=Curve, loc=<0 0 0>, rot=<0 0 0 1>}
 top_R_Corner_Up {type=Curve, loc=<0 0 0>, rot=<0 0 0 1>}
 bottom_L_Corner_Down {type=Curve, loc=<0 0 0>, rot=<0 0 0 1>}
 bottomt_L_Corner_Up {type=Curve, loc=<0 0 0>, rot=<0 0 0 1>}
 TrailCurve1 {type=Curve, loc=<0 0 0>, rot=<0 0 0 1>}
 BothHandFK {type=Curve, loc=<0 0 0>, rot=<0 0 0 1>}
 bottom_R_Corner_Down1 {type=Curve, loc=<0 0 0>, rot=<0 0 0 1>}
 blendPoint1 {type=Curve, loc=<0 0 0>, rot=<0 0 0 1>}
 WeaponAimingFreeze {type=Curve, loc=<0 0 0>, rot=<0 0 0 1>}
 l_calf_top_taper {type=Curve, loc=<0 0 0>, rot=<0 0 0 1>}
 JogBlendSpaceRotation {type=Curve, loc=<0 0 0>, rot=<0 0 0 1>}
 ZeroOutPitchWeight {type=Curve, loc=<0 0 0>, rot=<0 0 0 1>}
 MeleeTwist {type=Curve, loc=<0 0 0>, rot=<0 0 0 1>}
 MeleeAnim {type=Curve, loc=<0 0 0>, rot=<0 0 0 1>}
 l_sneer_fix {type=Curve, loc=<0 0 0>, rot=<0 0 0 1>}
 jaw_fix {type=Curve, loc=<0 0 0>, rot=<0 0 0 1>}
 r_sneer_fix {type=Curve, loc=<0 0 0>, rot=<0 0 0 1>}
 TurnRotationAmount {type=Curve, loc=<0 0 0>, rot=<0 0 0 1>}
 TurnEnd {type=Curve, loc=<0 0 0>, rot=<0 0 0 1>}
 BlendIn {type=Curve, loc=<0 0 0>, rot=<0 0 0 1>}
 BlendOut {type=Curve, loc=<0 0 0>, rot=<0 0 0 1>}
 BodyRotateX {type=Curve, loc=<0 0 0>, rot=<0 0 0 1>}
 BodyTranslateX {type=Curve, loc=<0 0 0>, rot=<0 0 0 1>}
 CrescentKickScale {type=Curve, loc=<0 0 0>, rot=<0 0 0 1>}
 mask_geo {type=Curve, loc=<0 0 0>, rot=<0 0 0 1>}
 RemoveJumpAdditive {type=Curve, loc=<0 0 0>, rot=<0 0 0 1>}
 RightLeg_IK {type=Curve, loc=<0 0 0>, rot=<0 0 0 1>}
 SnapWeapon_LHand {type=Curve, loc=<0 0 0>, rot=<0 0 0 1>}
 PivotAnim {type=Curve, loc=<0 0 0>, rot=<0 0 0 1>}
 DefaultAdditiveStateWeight {type=Curve, loc=<0 0 0>, rot=<0 0 0 1>}
 StartAnim {type=Curve, loc=<0 0 0>, rot=<0 0 0 1>}
 StopAnim {type=Curve, loc=<0 0 0>, rot=<0 0 0 1>}
 DisableSpeedWarping {type=Curve, loc=<0 0 0>, rot=<0 0 0 1>}
 DisableDBNOSlopeWarp {type=Curve, loc=<0 0 0>, rot=<0 0 0 1>}
 MaskRightArm {type=Curve, loc=<0 0 0>, rot=<0 0 0 1>}
 TurnInPlaceAnim {type=Curve, loc=<0 0 0>, rot=<0 0 0 1>}
 Head_MESH {type=Curve, loc=<0 0 0>, rot=<0 0 0 1>}
 Hair_MESH {type=Curve, loc=<0 0 0>, rot=<0 0 0 1>}
 LaunchPadTransitionOff {type=Curve, loc=<0 0 0>, rot=<0 0 0 1>}
 C_glabella_up_pose {type=Curve, loc=<0 0 0>, rot=<0 0 0 1>}
 jaw_open_pose {type=Curve, loc=<0 0 0>, rot=<0 0 0 1>}
 L_blink_pose {type=Curve, loc=<0 0 0>, rot=<0 0 0 1>}
 L_brow_down_pose {type=Curve, loc=<0 0 0>, rot=<0 0 0 1>}
 L_brow_squeeze_in_pose {type=Curve, loc=<0 0 0>, rot=<0 0 0 1>}
 L_brow_up_pose {type=Curve, loc=<0 0 0>, rot=<0 0 0 1>}
 L_frown_pose {type=Curve, loc=<0 0 0>, rot=<0 0 0 1>}
 L_lip_corner_narrow_pose {type=Curve, loc=<0 0 0>, rot=<0 0 0 1>}
 L_squint_inner_pose {type=Curve, loc=<0 0 0>, rot=<0 0 0 1>}
 L_wide_pose {type=Curve, loc=<0 0 0>, rot=<0 0 0 1>}
 look_down_pose {type=Curve, loc=<0 0 0>, rot=<0 0 0 1>}
 look_left_pose {type=Curve, loc=<0 0 0>, rot=<0 0 0 1>}
 look_right_pose {type=Curve, loc=<0 0 0>, rot=<0 0 0 1>}
 look_up_pose {type=Curve, loc=<0 0 0>, rot=<0 0 0 1>}
 phoneme_fv_pose {type=Curve, loc=<0 0 0>, rot=<0 0 0 1>}
 phoneme_oo_pose {type=Curve, loc=<0 0 0>, rot=<0 0 0 1>}
 R_blink_pose {type=Curve, loc=<0 0 0>, rot=<0 0 0 1>}
 R_brow_down_pose {type=Curve, loc=<0 0 0>, rot=<0 0 0 1>}
 R_brow_squeeze_in_pose {type=Curve, loc=<0 0 0>, rot=<0 0 0 1>}
 R_brow_up_pose {type=Curve, loc=<0 0 0>, rot=<0 0 0 1>}
 R_frown_pose {type=Curve, loc=<0 0 0>, rot=<0 0 0 1>}
 R_lip_corner_narrow_pose {type=Curve, loc=<0 0 0>, rot=<0 0 0 1>}
 R_squint_inner_pose {type=Curve, loc=<0 0 0>, rot=<0 0 0 1>}
 R_wide_pose {type=Curve, loc=<0 0 0>, rot=<0 0 0 1>}
 DisableArmsHeadAdditive {type=Curve, loc=<0 0 0>, rot=<0 0 0 1>}
 FrontendRotation {type=Curve, loc=<0 0 0>, rot=<0 0 0 1>}
 BothLegFK {type=Curve, loc=<0 0 0>, rot=<0 0 0 1>}
 AllowTurnInPlace {type=Curve, loc=<0 0 0>, rot=<0 0 0 1>}
 MontageTimeStretchCurve {type=Curve, loc=<0 0 0>, rot=<0 0 0 1>}
 ParaGlideLeanAlpha {type=Curve, loc=<0 0 0>, rot=<0 0 0 1>}
 WeaponRInChestSpace {type=Curve, loc=<0 0 0>, rot=<0 0 0 1>}
 JimBeam_AO {type=Curve, loc=<0 0 0>, rot=<0 0 0 1>}
 R_squeeze_pose {type=Curve, loc=<0 0 0>, rot=<0 0 0 1>}
 L_squeeze_pose {type=Curve, loc=<0 0 0>, rot=<0 0 0 1>}
 R_upper_lip_raiser_pose {type=Curve, loc=<0 0 0>, rot=<0 0 0 1>}
 L_cheek_up_pose {type=Curve, loc=<0 0 0>, rot=<0 0 0 1>}
 R_cheek_up_pose {type=Curve, loc=<0 0 0>, rot=<0 0 0 1>}
 DisableFaceOverride {type=Curve, loc=<0 0 0>, rot=<0 0 0 1>}
 EnableFullBodyOverride {type=Curve, loc=<0 0 0>, rot=<0 0 0 1>}
 C_glabella_down_pose {type=Curve, loc=<0 0 0>, rot=<0 0 0 1>}
 C_nose_wrinkler_pose {type=Curve, loc=<0 0 0>, rot=<0 0 0 1>}
 eye_convergence_pose {type=Curve, loc=<0 0 0>, rot=<0 0 0 1>}
 jaw_back_pose {type=Curve, loc=<0 0 0>, rot=<0 0 0 1>}
 jaw_forward_pose {type=Curve, loc=<0 0 0>, rot=<0 0 0 1>}
 jaw_left_pose {type=Curve, loc=<0 0 0>, rot=<0 0 0 1>}
 jaw_right_pose {type=Curve, loc=<0 0 0>, rot=<0 0 0 1>}
 jaw_up_pose {type=Curve, loc=<0 0 0>, rot=<0 0 0 1>}
 L_lip_corner_wide_pose {type=Curve, loc=<0 0 0>, rot=<0 0 0 1>}
 L_lower_lip_down_pose {type=Curve, loc=<0 0 0>, rot=<0 0 0 1>}
 L_lower_lip_up_pose {type=Curve, loc=<0 0 0>, rot=<0 0 0 1>}
 L_smile_pose {type=Curve, loc=<0 0 0>, rot=<0 0 0 1>}
 L_upper_lip_lower_pose {type=Curve, loc=<0 0 0>, rot=<0 0 0 1>}
 L_upper_lip_raiser_pose {type=Curve, loc=<0 0 0>, rot=<0 0 0 1>}
 L_upperLidDown_pose {type=Curve, loc=<0 0 0>, rot=<0 0 0 1>}
 mouth_down_pose {type=Curve, loc=<0 0 0>, rot=<0 0 0 1>}
 mouth_up_pose {type=Curve, loc=<0 0 0>, rot=<0 0 0 1>}
 phoneme_ch_pose {type=Curve, loc=<0 0 0>, rot=<0 0 0 1>}
 phoneme_mbp_pose {type=Curve, loc=<0 0 0>, rot=<0 0 0 1>}
 R_lip_corner_wide_pose {type=Curve, loc=<0 0 0>, rot=<0 0 0 1>}
 R_lower_lip_down_pose {type=Curve, loc=<0 0 0>, rot=<0 0 0 1>}
 R_lower_lip_up_pose {type=Curve, loc=<0 0 0>, rot=<0 0 0 1>}
 R_smile_pose {type=Curve, loc=<0 0 0>, rot=<0 0 0 1>}
 R_upper_lip_lower_pose {type=Curve, loc=<0 0 0>, rot=<0 0 0 1>}
 R_upperLidDown_pose {type=Curve, loc=<0 0 0>, rot=<0 0 0 1>}
 tongue_back_pose {type=Curve, loc=<0 0 0>, rot=<0 0 0 1>}
 tongue_down_pose {type=Curve, loc=<0 0 0>, rot=<0 0 0 1>}
 tongue_forward_pose {type=Curve, loc=<0 0 0>, rot=<0 0 0 1>}
 tongue_up_pose {type=Curve, loc=<0 0 0>, rot=<0 0 0 1>}
 LiveLink {type=Curve, loc=<0 0 0>, rot=<0 0 0 1>}
 ParaGlideDeploy {type=Curve, loc=<0 0 0>, rot=<0 0 0 1>}
 HideLeftWeapon {type=Curve, loc=<0 0 0>, rot=<0 0 0 1>}
 DisableLeftHandIKGolfCart {type=Curve, loc=<0 0 0>, rot=<0 0 0 1>}
 DisableRightHandIKGolfCart {type=Curve, loc=<0 0 0>, rot=<0 0 0 1>}
 DisableHandIKRetargeting {type=Curve, loc=<0 0 0>, rot=<0 0 0 1>}
 DisableConsumableOffsetPose {type=Curve, loc=<0 0 0>, rot=<0 0 0 1>}
 UpperBodyNoAdditive {type=Curve, loc=<0 0 0>, rot=<0 0 0 1>}
 MaskLeftArm {type=Curve, loc=<0 0 0>, rot=<0 0 0 1>}
 ForceUpperBody {type=Curve, loc=<0 0 0>, rot=<0 0 0 1>}
 RightHandFK {type=Curve, loc=<0 0 0>, rot=<0 0 0 1>}
 ParaGlideLeanTransition {type=Curve, loc=<0 0 0>, rot=<0 0 0 1>}
 InMotionEmote {type=Curve, loc=<0 0 0>, rot=<0 0 0 1>}
 ForceFullBodyInFrontEnd {type=Curve, loc=<0 0 0>, rot=<0 0 0 1>}
 WeaponLinFloorSpace {type=Curve, loc=<0 0 0>, rot=<0 0 0 1>}
 FrontEnd_Idle_Bypass {type=Curve, loc=<0 0 0>, rot=<0 0 0 1>}
 DisableSlopeWarping {type=Curve, loc=<0 0 0>, rot=<0 0 0 1>}
 Disable_Custom {type=Curve, loc=<0 0 0>, rot=<0 0 0 1>}
 EmoteRequiresWeapon {type=Curve, loc=<0 0 0>, rot=<0 0 0 1>}
 DisableIKRootAdditive {type=Curve, loc=<0 0 0>, rot=<0 0 0 1>}
 DisableFullBodySprint {type=Curve, loc=<0 0 0>, rot=<0 0 0 1>}
 DetachFeet_Vehicle {type=Curve, loc=<0 0 0>, rot=<0 0 0 1>}
 IndigoBoost {type=Curve, loc=<0 0 0>, rot=<0 0 0 1>}
 DisableUpperBodyAdditiveMeshSpace {type=Curve, loc=<0 0 0>, rot=<0 0 0 1>}
 MeleeAOWeight {type=Curve, loc=<0 0 0>, rot=<0 0 0 1>}
 HandsInRootSpace {type=Curve, loc=<0 0 0>, rot=<0 0 0 1>}
 CrouchPelvisAdjustment {type=Curve, loc=<0 0 0>, rot=<0 0 0 1>}
 EnableAccelYawCorrection {type=Curve, loc=<0 0 0>, rot=<0 0 0 1>}
 ParaGlideModifyRoot {type=Curve, loc=<0 0 0>, rot=<0 0 0 1>}
 UseScalarParameterForBlink {type=Curve, loc=<0 0 0>, rot=<0 0 0 1>}
 BlinkControl {type=Curve, loc=<0 0 0>, rot=<0 0 0 1>}
 ForceTargeting {type=Curve, loc=<0 0 0>, rot=<0 0 0 1>}
 BlinkControlProp {type=Curve, loc=<0 0 0>, rot=<0 0 0 1>}
 UseScalarParameterForBlinkProp {type=Curve, loc=<0 0 0>, rot=<0 0 0 1>}
 CrackshotCozyHeadCorrection {type=Curve, loc=<0 0 0>, rot=<0 0 0 1>}
 ForwardMotion {type=Curve, loc=<0 0 0>, rot=<0 0 0 1>}
 EmoteForcedRootYaw {type=Curve, loc=<0 0 0>, rot=<0 0 0 1>}
 MeleeTrailWidth {type=Curve, loc=<0 0 0>, rot=<0 0 0 1>}
 space_world {type=Curve, loc=<0 0 0>, rot=<0 0 0 1>}
 PeelyBlenderFace {type=Curve, loc=<0 0 0>, rot=<0 0 0 1>}
 Width1 {type=Curve, loc=<0 0 0>, rot=<0 0 0 1>}
 OverrideIKHands {type=Curve, loc=<0 0 0>, rot=<0 0 0 1>}
 ParaGlideAddAlpha {type=Curve, loc=<0 0 0>, rot=<0 0 0 1>}
 ParaGlideTurn {type=Curve, loc=<0 0 0>, rot=<0 0 0 1>}
 DisablePelvisAdjustment {type=Curve, loc=<0 0 0>, rot=<0 0 0 1>}
 UncleBrollyDeploy {type=Curve, loc=<0 0 0>, rot=<0 0 0 1>}
 ForceRootSlopeOrientation {type=Curve, loc=<0 0 0>, rot=<0 0 0 1>}
 DisableAdditiveOffsetPose {type=Curve, loc=<0 0 0>, rot=<0 0 0 1>}
 DisableIdleNoiseAdditive {type=Curve, loc=<0 0 0>, rot=<0 0 0 1>}
 UsingConsumable {type=Curve, loc=<0 0 0>, rot=<0 0 0 1>}
 DisableRightArmAdditive {type=Curve, loc=<0 0 0>, rot=<0 0 0 1>}
]
variables [
 KneeRot_fr_l_114deg {type=FVector, default=(X=0,Y=0,Z=0)}
 KneeRot_fr_l_130deg {type=FVector, default=(X=0,Y=0,Z=0)}
 KneeScale_fr_l_114deg {type=FVector, default=(X=0,Y=0,Z=0)}
 KneeScale_fr_l_130deg {type=FVector, default=(X=0,Y=0,Z=0)}
 KneeTrans_fr_l_114deg {type=FVector, default=(X=0,Y=0,Z=0)}
 KneeTrans_fr_l_130deg {type=FVector, default=(X=0,Y=0,Z=0)}
]
