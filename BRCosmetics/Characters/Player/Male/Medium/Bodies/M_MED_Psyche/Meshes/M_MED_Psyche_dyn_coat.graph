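# Physics asset for the dynamic coat on the male medium Psyche player body, simulating the jacket tails against the lower body.
# Six four-segment capsule chains (front, side, back per side) hang from master bodies linked to the pelvis, with tight twist limits and wider front swing toward the hem.
# Free constraints link neighboring hem segments on each side.
# Pelvis, leg, foot and a large crouch capsule provide body collision shapes.
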PhysicsAsset /BRCosmetics/Characters/Player/Male/Medium/Bodies/M_MED_Psyche/Meshes/M_MED_Psyche_dyn_coat
bodies {
 calf_l [{shapeType=Capsule, length=42, radius=7.25, center=<-24.0556 -2.1762 0.6745>, rotation=<rot 90 -10.1866 171.2769>}]
 calf_r [{shapeType=Capsule, length=42, radius=7.25, center=<24.0556 2.1762 -0.6745>, rotation=<rot 90 6.2419 7.7055>}]
 dyn_crouch_collision [{
   shapeType=Capsule
   length=20.25
   radius=36.5
   shapeName=dyn_crouch_collision_capsule
   center=<-0.0003 -9.4526 -15.283>
   rotation=<rot 90 0 0>
  }]
 dyn_jacketBack_01_l [{shapeType=Capsule, length=18.5, radius=7, center=<4 0 10>, rotation=<rot 0 0 90>}]
 dyn_jacketBack_01_r [{shapeType=Capsule, length=18.5, radius=7, center=<4 0 10>, rotation=<rot 0 0 90>}]
 dyn_jacketBack_02_l [{shapeType=Capsule, length=19.5, radius=7, center=<6 0 10>, rotation=<rot 0 0 90>}]
 dyn_jacketBack_02_r [{shapeType=Capsule, length=19.5, radius=7, center=<6 0 10>, rotation=<rot 0 0 90>}]
 dyn_jacketBack_03_l [{shapeType=Capsule, length=20, radius=10, center=<6 0 9.8>, rotation=<rot 0 0 90>}]
 dyn_jacketBack_03_r [{shapeType=Capsule, length=20, radius=10, center=<6 0 9.8>, rotation=<rot 0 0 90>}]
 dyn_jacketBack_04_l [{shapeType=Capsule, length=21, radius=10, center=<8 0 6.8>, rotation=<rot 0 0 90>}]
 dyn_jacketBack_04_r [{shapeType=Capsule, length=21, radius=10, center=<8 0 6.8>, rotation=<rot 0 0 90>}]
 dyn_jacketBack_mstr_l [{shapeType=Capsule, length=0.505, radius=0.505, rotation=<rot 0 -90 0>}]
 dyn_jacketBack_mstr_r [{shapeType=Capsule, length=0.505, radius=0.505, rotation=<rot 0 -90 0>}]
 dyn_jacketFront_01_l [{shapeType=Capsule, length=7.25, radius=7, center=<6 -4 11>, rotation=<rot 0 0 90>}]
 dyn_jacketFront_01_r [{shapeType=Capsule, length=7.5, radius=7, center=<6 4 11>, rotation=<rot 0 0 79.9995>}]
 dyn_jacketFront_02_l [{shapeType=Capsule, length=8.5, radius=7, center=<6 -4 10>, rotation=<rot 0 0 90>}]
 dyn_jacketFront_02_r [{shapeType=Capsule, length=8.5, radius=7, center=<6 4 10>, rotation=<rot 0 0 79.9999>}]
 dyn_jacketFront_03_l [{shapeType=Capsule, length=7, radius=9, center=<6 -4 10>, rotation=<rot 0 0 90>}]
 dyn_jacketFront_03_r [{shapeType=Capsule, length=6.75, radius=9.25, center=<6 4 10>, rotation=<rot 0 0 79.9995>}]
 dyn_jacketFront_04_l [{shapeType=Capsule, length=7.5, radius=9, center=<8 -4 10>, rotation=<rot 0 0 85.7763>}]
 dyn_jacketFront_04_r [{shapeType=Capsule, length=7.25, radius=9.25, center=<8 4 10>, rotation=<rot 0 0 79.9987>}]
 dyn_jacketFront_mstr_l [{shapeType=Capsule, length=0.505, radius=0.505, rotation=<rot 0 -90 0>}]
 dyn_jacketFront_mstr_r [{shapeType=Capsule, length=0.505, radius=0.505, rotation=<rot 0 -90 0>}]
 dyn_jacketSide_01_l [{shapeType=Capsule, length=19.25, radius=7, center=<4 -2 9>, rotation=<rot 0 0 80>}]
 dyn_jacketSide_01_r [{shapeType=Capsule, length=19.25, radius=7, center=<4 2 9>, rotation=<rot 0 0 100>}]
 dyn_jacketSide_02_l [{shapeType=Capsule, length=20, radius=7, center=<6 -2 10>, rotation=<rot 0 0 80>}]
 dyn_jacketSide_02_r [{shapeType=Capsule, length=19.75, radius=7, center=<6 2 10>, rotation=<rot 0 0 100>}]
 dyn_jacketSide_03_l [{shapeType=Capsule, length=20, radius=10, center=<6 -2 10>, rotation=<rot 0 0 80>}]
 dyn_jacketSide_03_r [{shapeType=Capsule, length=20, radius=10, center=<6 2 10>, rotation=<rot 0 0 100>}]
 dyn_jacketSide_04_l [{shapeType=Capsule, length=21, radius=10, center=<8 -2 10>, rotation=<rot 0 0 80>}]
 dyn_jacketSide_04_r [{shapeType=Capsule, length=21, radius=10, center=<8 2 10>, rotation=<rot 0 0 100>}]
 dyn_jacketSide_mstr_l [{shapeType=Capsule, length=0.505, radius=0.505, rotation=<rot 0 -90 0>}]
 dyn_jacketSide_mstr_r [{shapeType=Capsule, length=0.505, radius=0.505, rotation=<rot 0 -90 0>}]
 foot_l [{shapeType=Capsule, length=22.25, radius=9.5, center=<-3.5001 8.3295 0.0009>, rotation=<rot 0 0 -90>}]
 foot_r [{shapeType=Capsule, length=22.25, radius=9.5, center=<3.5001 -8.3295 -0.0009>, rotation=<rot 0 0 90>}]
 pelvis [{shapeType=Capsule, length=9.25, radius=14, center=<-2 0 0>}]
 thigh_l [{shapeType=Capsule, length=35.25, radius=10, center=<-18.8837 0.0009 0>, rotation=<rot 86.4154 0 -2.3844>}]
 thigh_r [{shapeType=Capsule, length=35.25, radius=10, center=<18.8837 -0.0009 0>, rotation=<rot 86.4154 0 177.6156>}]
}
constraints [
 {
  bone1Name=thigh_l
  bone2Name=pelvis
  swing1LimitDegrees=45
  swing1Motion=Free
  swing2LimitDegrees=45
  swing2Motion=Free
  twistLimitDegrees=45
  twistMotion=Free
 }
 {bone1Name=dyn_jacketSide_01_l, bone2Name=dyn_jacketSide_mstr_l, swing1LimitDegrees=10, swing2LimitDegrees=45, twistLimitDegrees=2}
 {bone1Name=dyn_jacketSide_02_l, bone2Name=dyn_jacketSide_01_l, swing1LimitDegrees=10, swing2LimitDegrees=45, twistLimitDegrees=2}
 {bone1Name=dyn_jacketSide_03_l, bone2Name=dyn_jacketSide_02_l, swing1LimitDegrees=10, swing2LimitDegrees=45, twistLimitDegrees=2}
 {bone1Name=dyn_jacketSide_04_l, bone2Name=dyn_jacketSide_03_l, swing1LimitDegrees=10, swing2LimitDegrees=45, twistLimitDegrees=2}
 {bone1Name=dyn_jacketBack_01_l, bone2Name=dyn_jacketBack_mstr_l, swing1LimitDegrees=10, swing2LimitDegrees=45, twistLimitDegrees=2}
 {bone1Name=dyn_jacketBack_02_l, bone2Name=dyn_jacketBack_01_l, swing1LimitDegrees=10, swing2LimitDegrees=45, twistLimitDegrees=2}
 {bone1Name=dyn_jacketBack_03_l, bone2Name=dyn_jacketBack_02_l, swing1LimitDegrees=10, swing2LimitDegrees=45, twistLimitDegrees=2}
 {bone1Name=dyn_jacketBack_04_l, bone2Name=dyn_jacketBack_03_l, swing1LimitDegrees=10, swing2LimitDegrees=45, twistLimitDegrees=2}
 {bone1Name=dyn_jacketBack_01_r, bone2Name=dyn_jacketBack_mstr_r, swing1LimitDegrees=10, swing2LimitDegrees=45, twistLimitDegrees=2}
 {bone1Name=dyn_jacketBack_02_r, bone2Name=dyn_jacketBack_01_r, swing1LimitDegrees=10, swing2LimitDegrees=45, twistLimitDegrees=2}
 {bone1Name=dyn_jacketBack_03_r, bone2Name=dyn_jacketBack_02_r, swing1LimitDegrees=10, swing2LimitDegrees=45, twistLimitDegrees=2}
 {bone1Name=dyn_jacketBack_04_r, bone2Name=dyn_jacketBack_03_r, swing1LimitDegrees=10, swing2LimitDegrees=45, twistLimitDegrees=2}
 {bone1Name=dyn_jacketSide_01_r, bone2Name=dyn_jacketSide_mstr_r, swing1LimitDegrees=10, swing2LimitDegrees=45, twistLimitDegrees=2}
 {bone1Name=dyn_jacketSide_02_r, bone2Name=dyn_jacketSide_01_r, swing1LimitDegrees=10, swing2LimitDegrees=45, twistLimitDegrees=2}
 {bone1Name=dyn_jacketSide_03_r, bone2Name=dyn_jacketSide_02_r, swing1LimitDegrees=10, swing2LimitDegrees=45, twistLimitDegrees=2}
 {bone1Name=dyn_jacketSide_04_r, bone2Name=dyn_jacketSide_03_r, swing1LimitDegrees=10, swing2LimitDegrees=45, twistLimitDegrees=2}
 {
  bone1Name=dyn_jacketSide_mstr_l
  bone2Name=pelvis
  swing1LimitDegrees=45
  swing1Motion=Free
  swing2LimitDegrees=45
  swing2Motion=Free
  twistLimitDegrees=45
  twistMotion=Free
 }
 {
  bone1Name=dyn_jacketBack_mstr_l
  bone2Name=pelvis
  swing1LimitDegrees=45
  swing1Motion=Free
  swing2LimitDegrees=45
  swing2Motion=Free
  twistLimitDegrees=45
  twistMotion=Free
 }
 {
  bone1Name=dyn_jacketBack_mstr_r
  bone2Name=pelvis
  swing1LimitDegrees=45
  swing1Motion=Free
  swing2LimitDegrees=45
  swing2Motion=Free
  twistLimitDegrees=45
  twistMotion=Free
 }
 {
  bone1Name=dyn_jacketSide_mstr_r
  bone2Name=pelvis
  swing1LimitDegrees=45
  swing1Motion=Free
  swing2LimitDegrees=45
  swing2Motion=Free
  twistLimitDegrees=45
  twistMotion=Free
 }
 {
  bone1Name=dyn_jacketFront_mstr_l
  bone2Name=pelvis
  swing1LimitDegrees=45
  swing1Motion=Free
  swing2LimitDegrees=45
  swing2Motion=Free
  twistLimitDegrees=45
  twistMotion=Free
 }
 {bone1Name=dyn_jacketFront_01_l, bone2Name=dyn_jacketFront_mstr_l, swing1LimitDegrees=5, swing2LimitDegrees=45, twistLimitDegrees=5}
 {bone1Name=dyn_jacketFront_02_l, bone2Name=dyn_jacketFront_01_l, swing1LimitDegrees=5, swing2LimitDegrees=60, twistLimitDegrees=5}
 {bone1Name=dyn_jacketFront_03_l, bone2Name=dyn_jacketFront_02_l, swing1LimitDegrees=5, swing2LimitDegrees=75, twistLimitDegrees=5}
 {bone1Name=dyn_jacketFront_04_l, bone2Name=dyn_jacketFront_03_l, swing1LimitDegrees=5, swing2LimitDegrees=75, twistLimitDegrees=5}
 {
  bone1Name=dyn_jacketFront_mstr_r
  bone2Name=pelvis
  swing1LimitDegrees=45
  swing1Motion=Free
  swing2LimitDegrees=45
  swing2Motion=Free
  twistLimitDegrees=45
  twistMotion=Free
 }
 {bone1Name=dyn_jacketFront_01_r, bone2Name=dyn_jacketFront_mstr_r, swing1LimitDegrees=5, swing2LimitDegrees=45, twistLimitDegrees=5}
 {bone1Name=dyn_jacketFront_02_r, bone2Name=dyn_jacketFront_01_r, swing1LimitDegrees=5, swing2LimitDegrees=60, twistLimitDegrees=5}
 {bone1Name=dyn_jacketFront_03_r, bone2Name=dyn_jacketFront_02_r, swing1LimitDegrees=5, swing2LimitDegrees=75, twistLimitDegrees=5}
 {bone1Name=dyn_jacketFront_04_r, bone2Name=dyn_jacketFront_03_r, swing1LimitDegrees=5, swing2LimitDegrees=75, twistLimitDegrees=5}
 {
  bone1Name=dyn_jacketSide_04_r
  bone2Name=dyn_jacketBack_04_r
  swing1LimitDegrees=45
  swing1Motion=Free
  swing2LimitDegrees=45
  swing2Motion=Free
  twistLimitDegrees=45
  twistMotion=Free
 }
 {
  bone1Name=foot_r
  bone2Name=calf_r
  swing1LimitDegrees=45
  swing1Motion=Free
  swing2LimitDegrees=45
  swing2Motion=Free
  twistLimitDegrees=45
  twistMotion=Free
 }
 {
  bone1Name=dyn_jacketSide_04_l
  bone2Name=dyn_jacketBack_04_l
  swing1LimitDegrees=45
  swing1Motion=Free
  swing2LimitDegrees=45
  swing2Motion=Free
  twistLimitDegrees=45
  twistMotion=Free
 }
 {
  bone1Name=dyn_jacketFront_04_r
  bone2Name=dyn_jacketSide_04_r
  swing1LimitDegrees=45
  swing1Motion=Free
  swing2LimitDegrees=45
  swing2Motion=Free
  twistLimitDegrees=45
  twistMotion=Free
 }
 {
  bone1Name=dyn_jacketFront_04_l
  bone2Name=dyn_jacketSide_04_l
  swing1LimitDegrees=45
  swing1Motion=Free
  swing2LimitDegrees=45
  swing2Motion=Free
  twistLimitDegrees=45
  twistMotion=Free
 }
 {
  bone1Name=foot_l
  bone2Name=calf_l
  swing1LimitDegrees=45
  swing1Motion=Free
  swing2LimitDegrees=45
  swing2Motion=Free
  twistLimitDegrees=45
  twistMotion=Free
 }
 {bone1Name=calf_l, bone2Name=thigh_l, swing1LimitDegrees=10, swing2LimitDegrees=45, twistLimitDegrees=2}
 {bone1Name=calf_r, bone2Name=thigh_r, swing1LimitDegrees=10, swing2LimitDegrees=45, twistLimitDegrees=2}
 {
  bone1Name=thigh_r
  bone2Name=pelvis
  swing1LimitDegrees=45
  swing1Motion=Free
  swing2LimitDegrees=45
  swing2Motion=Free
  twistLimitDegrees=45
  twistMotion=Free
 }
 {
  bone1Name=thigh_l
  bone2Name=thigh_r
  swing1LimitDegrees=45
  swing1Motion=Free
  swing2LimitDegrees=45
  swing2Motion=Free
  twistLimitDegrees=45
  twistMotion=Free
 }
]
