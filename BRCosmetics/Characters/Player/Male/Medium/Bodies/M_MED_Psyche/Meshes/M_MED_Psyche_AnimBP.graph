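# Animation Blueprint for the M_MED_Psyche body that copies the parent mesh pose, applies deform Control Rig and head-offset corrections, and animates the coat.
# With rigid bodies enabled, the coat uses RigidBody simulation with gravity override.
# Otherwise, crouch-dependent bone offsets and standing-pose Look Ats feed movement-weighted Trail controllers.
# The EventGraph derives blend alphas from thigh rotations, jumping and movement, zeroing back trail while rising in skydive.
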AnimBlueprint /BRCosmetics/Characters/Player/Male/Medium/Bodies/M_MED_Psyche/Meshes/M_MED_Psyche_AnimBP
target_skeleton=&/M_MED_Psyche_Skeleton
graphs [
 AnimGraph {nodes [
   Output Pose(Result=Blend Poses by bool(BlendPose_0=Use cached pose 'rigidPose'(), BlendPose_1=Use cached pose 'unRigidPose'(), bActiveValue=Get bRigidBodyEnabled()) {blendType=Linear}) {layerGroup=None}
   n1 = Control Rig!(
    Source=Copy Pose From Mesh() {bUseAttachedParent=true}
    Alpha
    KneeRot_fr_l_114deg:Vector
    KneeTrans_fr_l_114deg:Vector="0.258,-3.591,0.009"
    KneeScale_fr_l_114deg:Vector="1,1,1"
    KneeScale_fr_l_130deg:Vector="1,1,1"
    KneeRot_fr_l_130deg:Vector="0,0,-51.2"
    KneeTrans_fr_l_130deg:Vector="0.258,-3.591,0.009"
   ) -> (Pose) {
    alphaBoolBlend {blendOption=Linear}
    controlRigAssetReference {blueprintRigClass=&/M_MED_Psyche_Deform_CtrlRig_sculpters.M_MED_Psyche_Deform_CtrlRig_sculpters_C}
   }
   n2 = Copy Bone - Source Bone: head - Target Bone: head_offset!(
    ComponentPose:Component Space Pose Link Structure=Local To Component(LocalPose=@n1)
    Alpha:float=0.65
    bCopyTranslation:bool=False
    bCopyRotation:bool=true
    bCopyScale:bool=False
   ) -> (Pose:Component Space Pose Link Structure) {
    is=title AnimGraphNode_CopyBone
    controlSpace=BCS_BoneSpace
    alphaBoolBlend {blendOption=Linear}
    sourceBone {boneName=head}
    targetBone {boneName=head_offset}
   }
   Save cached pose 'prePose'(Pose=Component To Local(ComponentPose=@n2)) {cacheName=prePose}
   n3 = "Transform (Modify) Bone - Bone: dyn_jacketFront_mstr_l"(
    ComponentPose=Local To Component(LocalPose=Use cached pose 'prePose'())
    Alpha=Get thigh_rot_up_l() -> (thigh_rot_up_l:double)
    Translation="-2,0,0"
    Rotation="35,-30,-70"
    Scale="1,1,1"
   ) {
    rotationMode=BMM_Additive
    rotationSpace=BCS_BoneSpace
    translationMode=BMM_Additive
    translationSpace=BCS_BoneSpace
    alphaBoolBlend {blendOption=Linear}
    boneToModify {boneName=dyn_jacketFront_mstr_l}
    rotation=<rot 35 -30 -70>
    translation {x=-2}
   }
   n4 = "Transform (Modify) Bone - Bone: dyn_jacketFront_mstr_r"(ComponentPose=@n3, Alpha=Get thigh_rot_up_r() -> (thigh_rot_up_r:double), Translation="-2,0,0", Rotation="35,30,60", Scale="1,1,1") {
    rotationMode=BMM_Additive
    rotationSpace=BCS_BoneSpace
    translationMode=BMM_Additive
    translationSpace=BCS_BoneSpace
    alphaBoolBlend {blendOption=Linear}
    boneToModify {boneName=dyn_jacketFront_mstr_r}
    rotation=<rot 35 30 60>
    translation {x=-2}
   }
   n5 = RigidBody!(ComponentPose=@n4, Alpha, OverrideWorldGravity:Vector=Get GravityOverride(), ExternalForce) -> (Pose) {
    bClampLinearTranslationLimitToRefPose=true
    bOverrideWorldGravity=true
    overridePhysicsAsset=&/M_MED_Psyche_dyn_coat
    simulationSpace=BaseBoneSpace
    alphaBoolBlend {blendOption=Linear}
    baseBoneRef {boneName=spine_02}
    componentAppliedLinearAccClamp=<3000 3000 3000>
    componentLinearAccScale {x=0.45}
    componentLinearVelScale {x=0.45}
    overrideWorldGravity {z=-980}
    simSpaceSettings {
     maxAngularAcceleration=200
     maxAngularVelocity=5
     maxLinearAcceleration=1000
     maxLinearVelocity=200
     velocityScaleZ=0
     worldAlpha=1
     externalLinearDragV=<1 1>
    }
   }
   Save cached pose 'rigidPose'(Pose=Component To Local(ComponentPose=@n5)) {cacheName=rigidPose}
   n6 = Get thigh_rot_inr_r() -> (thigh_rot_inr_r:double)
   n7 = Look At - Bone: dyn_jacketFront_mstr_r(ComponentPose=Local To Component(LocalPose=Use cached pose 'prePose'()), Alpha=@n6) {
    alphaBoolBlend {blendOption=Linear}
    alphaScaleBiasClamp {bInterpResult=true}
    boneToModify {boneName=dyn_jacketFront_mstr_r}
    lookAtLocation=<0 -12 5>
    lookAtTarget {boneReference {boneName=calf_r}}
    lookAt_Axis {axis=<1 0>}
    lookUp_Axis {axis {x=0}}
   }
   n8 = Look At - Bone: dyn_jacketSide_mstr_r(ComponentPose=@n7, Alpha=@n6) {
    alphaBoolBlend {blendOption=Linear}
    alphaScaleBiasClamp {bInterpResult=true}
    boneToModify {boneName=dyn_jacketSide_mstr_r}
    lookAtLocation {x=0, z=11}
    lookAtTarget {boneReference {boneName=calf_r}}
    lookAt_Axis {axis=<1 0>}
    lookUp_Axis {axis {x=0}}
   }
   n9 = Get thigh_rot_up_r()
   n10 = "Transform (Modify) Bone - Bone: dyn_jacketFront_01_r"(ComponentPose=@n8, Alpha=@n9, Translation="0,0,0", Rotation="9.096,4.987,10.829", Scale="1,1,1") {
    rotationMode=BMM_Additive
    rotationSpace=BCS_BoneSpace
    translationMode=BMM_Additive
    translationSpace=BCS_BoneSpace
    alphaBoolBlend {blendOption=Linear}
    boneToModify {boneName=dyn_jacketFront_01_r}
    rotation=<rot 9.096 4.987 10.829>
   }
   n11 = "Transform (Modify) Bone - Bone: dyn_jacketSide_01_r"(ComponentPose=@n10, Alpha=@n9, Translation="0,0,0", Rotation="5.000,0.000,0.000", Scale="1,1,1") {
    rotationMode=BMM_Additive
    rotationSpace=BCS_BoneSpace
    translationMode=BMM_Additive
    translationSpace=BCS_BoneSpace
    alphaBoolBlend {blendOption=Linear}
    boneToModify {boneName=dyn_jacketSide_01_r}
    rotation {pitch=5}
   }
   n12 = Get thigh_rot_inr_l() -> (thigh_rot_inr_l:double)
   n13 = Look At - Bone: dyn_jacketFront_mstr_l(ComponentPose=@n11, Alpha=@n12) {
    alphaBoolBlend {blendOption=Linear}
    alphaScaleBiasClamp {bInterpResult=true}
    boneToModify {boneName=dyn_jacketFront_mstr_l}
    lookAtLocation=<0 12 -5>
    lookAtTarget {boneReference {boneName=calf_l}}
    lookAt_Axis {axis=<1 0>}
    lookUp_Axis {axis {x=0}}
   }
   n14 = Look At - Bone: dyn_jacketSide_mstr_l(ComponentPose=@n13, Alpha=@n12) {
    alphaBoolBlend {blendOption=Linear}
    alphaScaleBiasClamp {bInterpResult=true}
    boneToModify {boneName=dyn_jacketSide_mstr_l}
    lookAtLocation {x=0, z=-11}
    lookAtTarget {boneReference {boneName=calf_l}}
    lookAt_Axis {axis=<1 0>}
    lookUp_Axis {axis {x=0}}
   }
   n15 = Get thigh_rot_up_l()
   n16 = "Transform (Modify) Bone - Bone: dyn_jacketFront_01_l"(ComponentPose=@n14, Alpha=@n15, Translation="0,0,0", Rotation="9.096,-4.987,-10.828", Scale="1,1,1") {
    rotationMode=BMM_Additive
    rotationSpace=BCS_BoneSpace
    translationMode=BMM_Additive
    translationSpace=BCS_BoneSpace
    alphaBoolBlend {blendOption=Linear}
    boneToModify {boneName=dyn_jacketFront_01_l}
    rotation=<rot 9.096 -4.987 -10.8284>
   }
   n17 = "Transform (Modify) Bone - Bone: dyn_jacketSide_01_l"(ComponentPose=@n16, Alpha=@n15, Translation="0,0,0", Rotation="5,0,0", Scale="1,1,1") {
    rotationMode=BMM_Additive
    rotationSpace=BCS_BoneSpace
    translationMode=BMM_Additive
    translationSpace=BCS_BoneSpace
    alphaBoolBlend {blendOption=Linear}
    boneToModify {boneName=dyn_jacketSide_01_l}
    rotation {pitch=5}
   }
   n18 = Get jumpingAlpha() -> (jumpingAlpha:double)
   n19 = "Transform (Modify) Bone - Bone: dyn_jacketBack_01_l"(ComponentPose=@n17, Alpha=@n18, Translation="0,0,0", Rotation="35,10,0.000", Scale="1,1,1") {
    rotationMode=BMM_Additive
    rotationSpace=BCS_BoneSpace
    alphaBoolBlend {blendOption=Linear}
    alphaScaleBiasClamp {bInterpResult=true, interpSpeedDecreasing=2, interpSpeedIncreasing=5}
    boneToModify {boneName=dyn_jacketBack_01_l}
    rotation {pitch=35, yaw=10}
   }
   n20 = "Transform (Modify) Bone - Bone: dyn_jacketBack_01_r"(ComponentPose=@n19, Alpha=@n18, Translation="0,0,0", Rotation="35,-10,0.000", Scale="1,1,1") {
    rotationMode=BMM_Additive
    rotationSpace=BCS_BoneSpace
    alphaBoolBlend {blendOption=Linear}
    alphaScaleBiasClamp {bInterpResult=true, interpSpeedDecreasing=2, interpSpeedIncreasing=5}
    boneToModify {boneName=dyn_jacketBack_01_r}
    rotation {pitch=35, yaw=-10}
   }
   n21 = "Transform (Modify) Bone - Bone: dyn_jacketSide_01_l"(ComponentPose=@n20, Alpha=@n18, Translation="0,0,0", Rotation="15,-15,0.000", Scale="1,1,1") {
    rotationMode=BMM_Additive
    rotationSpace=BCS_BoneSpace
    alphaBoolBlend {blendOption=Linear}
    alphaScaleBiasClamp {bInterpResult=true, interpSpeedDecreasing=2, interpSpeedIncreasing=5}
    boneToModify {boneName=dyn_jacketSide_01_l}
    rotation {pitch=15, yaw=-15}
   }
   n22 = "Transform (Modify) Bone - Bone: dyn_jacketSide_01_l"(ComponentPose=@n21, Alpha=@n18, Translation="0,0,0", Rotation="15,15,0.000", Scale="1,1,1") {
    rotationMode=BMM_Additive
    rotationSpace=BCS_BoneSpace
    alphaBoolBlend {blendOption=Linear}
    alphaScaleBiasClamp {bInterpResult=true, interpSpeedDecreasing=2, interpSpeedIncreasing=5}
    boneToModify {boneName=dyn_jacketSide_01_l}
    rotation {pitch=15, yaw=15}
   }
   n23 = Component To Local(ComponentPose=@n22)
   n24 = "Transform (Modify) Bone - Bone: dyn_jacketFront_01_r"(ComponentPose=Local To Component(LocalPose=Use cached pose 'prePose'()), Alpha=1, Translation="0,0,0", Rotation="15,0,0", Scale="1,1,1") {
    rotationMode=BMM_Additive
    rotationSpace=BCS_BoneSpace
    translationMode=BMM_Additive
    translationSpace=BCS_BoneSpace
    alphaBoolBlend {blendOption=Linear}
    boneToModify {boneName=dyn_jacketFront_01_r}
    rotation {pitch=15}
   }
   n25 = "Transform (Modify) Bone - Bone: dyn_jacketSide_01_r"(ComponentPose=@n24, Alpha=1, Translation="0,0,0", Rotation="20,0,0", Scale="1,1,1") {
    rotationMode=BMM_Additive
    rotationSpace=BCS_BoneSpace
    translationMode=BMM_Additive
    translationSpace=BCS_BoneSpace
    alphaBoolBlend {blendOption=Linear}
    boneToModify {boneName=dyn_jacketSide_01_r}
    rotation {pitch=20}
   }
   n26 = Get isNotCrouchRunning() -> (isNotCrouchRunning:double)
   n27 = "Transform (Modify) Bone - Bone: dyn_jacketBack_01_l"(ComponentPose=@n25, Alpha=@n26, Translation="0,0,0", Rotation="10,10,0.000", Scale="1,1,1") {
    rotationMode=BMM_Additive
    rotationSpace=BCS_BoneSpace
    alphaBoolBlend {blendOption=Linear}
    alphaScaleBiasClamp {bInterpResult=true, interpSpeedDecreasing=5}
    boneToModify {boneName=dyn_jacketBack_01_l}
    rotation {pitch=10, yaw=10}
   }
   n28 = "Transform (Modify) Bone - Bone: dyn_jacketBack_01_r"(ComponentPose=@n27, Alpha=@n26, Translation="0,0,0", Rotation="10,-10,0.000", Scale="1,1,1") {
    rotationMode=BMM_Additive
    rotationSpace=BCS_BoneSpace
    alphaBoolBlend {blendOption=Linear}
    alphaScaleBiasClamp {bInterpResult=true, interpSpeedDecreasing=5}
    boneToModify {boneName=dyn_jacketBack_01_r}
    rotation {pitch=10, yaw=-10}
   }
   n29 = "Transform (Modify) Bone - Bone: dyn_jacketBack_03_r"(ComponentPose=@n28, Alpha=@n26, Translation="0,0,0", Rotation="60,0,0.000", Scale="1,1,1") {
    rotationMode=BMM_Additive
    rotationSpace=BCS_BoneSpace
    alphaBoolBlend {blendOption=Linear}
    alphaScaleBiasClamp {bInterpResult=true, interpSpeedDecreasing=5}
    boneToModify {boneName=dyn_jacketBack_03_r}
    rotation {pitch=60}
   }
   n30 = "Transform (Modify) Bone - Bone: dyn_jacketFront_mstr_r"(ComponentPose=@n29, Alpha=Get thigh_rot_up_r(), Translation="0,0,3", Rotation="0,0,0", Scale="1,1,1") {
    rotationMode=BMM_Additive
    rotationSpace=BCS_BoneSpace
    translationMode=BMM_Additive
    translationSpace=BCS_BoneSpace
    alphaBoolBlend {blendOption=Linear}
    boneToModify {boneName=dyn_jacketFront_mstr_r}
    translation {z=3}
   }
   n31 = Get thigh_rot_otr_r() -> (thigh_rot_otr_r:double)
   n32 = "Transform (Modify) Bone - Bone: dyn_jacketFront_01_r"(ComponentPose=@n30, Alpha=@n31, Translation="0,0,0", Rotation="45,0,0", Scale="1,1,1") {
    rotationMode=BMM_Additive
    rotationSpace=BCS_BoneSpace
    translationMode=BMM_Additive
    translationSpace=BCS_BoneSpace
    alphaBoolBlend {blendOption=Linear}
    boneToModify {boneName=dyn_jacketFront_01_r}
    rotation {pitch=45}
   }
   n33 = "Transform (Modify) Bone - Bone: dyn_jacketFront_01_r"(ComponentPose=@n32, Alpha=@n31, Translation="0,0,2", Rotation="30,0,0", Scale="1,1,1") {
    rotationMode=BMM_Additive
    rotationSpace=BCS_BoneSpace
    translationMode=BMM_Additive
    translationSpace=BCS_BoneSpace
    alphaBoolBlend {blendOption=Linear}
    boneToModify {boneName=dyn_jacketFront_01_r}
    rotation {pitch=30}
    translation {z=2}
   }
   n34 = "Transform (Modify) Bone - Bone: dyn_jacketSide_02_r"(ComponentPose=@n33, Alpha=@n31, Translation="0,0,0", Rotation="45,0,0", Scale="1,1,1") {
    rotationMode=BMM_Additive
    rotationSpace=BCS_BoneSpace
    translationMode=BMM_Additive
    translationSpace=BCS_BoneSpace
    alphaBoolBlend {blendOption=Linear}
    boneToModify {boneName=dyn_jacketSide_02_r}
    rotation {pitch=45}
   }
   n35 = Get thigh_rot_up_l()
   n36 = "Transform (Modify) Bone - Bone: dyn_jacketFront_01_l"(ComponentPose=@n34, Alpha=@n35, Translation="0,0,0", Rotation="30,-15,-35", Scale="1,1,1") {
    rotationMode=BMM_Additive
    rotationSpace=BCS_BoneSpace
    translationMode=BMM_Additive
    translationSpace=BCS_BoneSpace
    alphaBoolBlend {blendOption=Linear}
    boneToModify {boneName=dyn_jacketFront_01_l}
    rotation=<rot 30 -15 -35>
   }
   n37 = "Transform (Modify) Bone - Bone: dyn_jacketFront_02_l"(ComponentPose=@n36, Alpha=@n35, Translation="0,0,0", Rotation="-45,0,0", Scale="1,1,1") {
    rotationMode=BMM_Additive
    rotationSpace=BCS_BoneSpace
    translationMode=BMM_Additive
    translationSpace=BCS_BoneSpace
    alphaBoolBlend {blendOption=Linear}
    boneToModify {boneName=dyn_jacketFront_02_l}
    rotation {pitch=-45}
   }
   n38 = "Transform (Modify) Bone - Bone: dyn_jacketSide_01_l"(ComponentPose=@n37, Alpha=@n35, Translation="0,-2,3", Rotation="0,-10,0", Scale="1,1,1") {
    rotationMode=BMM_Additive
    rotationSpace=BCS_BoneSpace
    translationMode=BMM_Additive
    translationSpace=BCS_BoneSpace
    alphaBoolBlend {blendOption=Linear}
    boneToModify {boneName=dyn_jacketSide_01_l}
    rotation {yaw=-10}
    translation {y=-2, z=3}
   }
   n39 = Get thigh_rot_otr_l() -> (thigh_rot_otr_l:double)
   n40 = "Transform (Modify) Bone - Bone: dyn_jacketFront_01_l"(ComponentPose=@n38, Alpha=@n39, Translation="0,0,0", Rotation="45,0,0", Scale="1,1,1") {
    rotationMode=BMM_Additive
    rotationSpace=BCS_BoneSpace
    translationMode=BMM_Additive
    translationSpace=BCS_BoneSpace
    alphaBoolBlend {blendOption=Linear}
    boneToModify {boneName=dyn_jacketFront_01_l}
    rotation {pitch=45}
   }
   n41 = "Transform (Modify) Bone - Bone: dyn_jacketFront_02_l"(ComponentPose=@n40, Alpha=@n39, Translation="0,0,2", Rotation="30,0,0", Scale="1,1,1") {
    rotationMode=BMM_Additive
    rotationSpace=BCS_BoneSpace
    translationMode=BMM_Additive
    translationSpace=BCS_BoneSpace
    alphaBoolBlend {blendOption=Linear}
    boneToModify {boneName=dyn_jacketFront_02_l}
    rotation {pitch=30}
    translation {z=2}
   }
   n42 = "Transform (Modify) Bone - Bone: dyn_jacketSide_02_l"(ComponentPose=@n41, Alpha=@n39, Translation="0,0,0", Rotation="45,0,0", Scale="1,1,1") {
    rotationMode=BMM_Additive
    rotationSpace=BCS_BoneSpace
    translationMode=BMM_Additive
    translationSpace=BCS_BoneSpace
    alphaBoolBlend {blendOption=Linear}
    boneToModify {boneName=dyn_jacketSide_02_l}
    rotation {pitch=45}
   }
   n43 = Component To Local(ComponentPose=@n42)
   n44 = Get bIsCrouching()
   n45 = Get BwdMovement() -> (BwdMovement:double)
   n46 = Trail controller - Bone: dyn_jacketBack_04_r!(
    ComponentPose=Local To Component(LocalPose=Blend Poses by bool(BlendPose_0=@n43, BlendPose_1=@n23, bActiveValue=@n44) {blendType=Linear})
    Alpha=@n45
    FakeVelocity:Vector="0,0,0"
   ) -> (Pose) {
    bActorSpaceFakeVel=true
    bShowTrailLocation=true
    chainLength=3
    debugLifeTime=0.1
    maxDeltaTime=0.0333
    relaxationSpeedScale=3
    alphaScaleBiasClamp {bInterpResult=true}
    baseJoint {boneName=root}
    planarLimits [
     {drivingBone {boneName=root}, planeTransform {location=<0 0 10>, rotation=<rot 0 0 0>, scale=<1 1 1>}}
     {drivingBone {boneName=spine_03}, planeTransform {location=<0 -20 0>, rotation=<rot 0 15 -89.9999>, scale=<1 1 1>}}
    ]
    rotationLimits [{limitMax=<180 180 180>}, {limitMax=<180 180 180>}, {limitMax=<180 180 180>}]
    rotationOffsets [<0 0 0>, <0 0 0>, <0 0 0>]
    trailBone {boneName=dyn_jacketBack_04_r}
    trailRelaxationSpeed {editorCurveData {keys [{arriveTangent=-72.0739, interpMode=RCIM_Cubic, leaveTangent=-72.0739, tangentMode=RCTM_User, time=0, value=35}, <at 1 5 RCIM_Cubic>]}}
   }
   n47 = Trail controller - Bone: dyn_jacketBack_04_l!(ComponentPose=@n46, Alpha=@n45, FakeVelocity:Vector="0,0,0") -> (Pose) {
    bActorSpaceFakeVel=true
    bShowTrailLocation=true
    chainLength=3
    debugLifeTime=0.1
    maxDeltaTime=0.0333
    relaxationSpeedScale=3
    alphaScaleBiasClamp {bInterpResult=true}
    baseJoint {boneName=root}
    planarLimits [
     {drivingBone {boneName=root}, planeTransform {location=<0 0 10>, rotation=<rot 0 0 0>, scale=<1 1 1>}}
     {drivingBone {boneName=spine_03}, planeTransform {location=<0 -20 0>, rotation=<rot 0 15 -89.9999>, scale=<1 1 1>}}
    ]
    rotationLimits [{limitMax=<180 180 180>}, {limitMax=<180 180 180>}, {limitMax=<180 180 180>}]
    rotationOffsets [<0 0 0>, <0 0 0>, <0 0 0>]
    trailBone {boneName=dyn_jacketBack_04_l}
    trailRelaxationSpeed {editorCurveData {keys [{arriveTangent=-72.0739, interpMode=RCIM_Cubic, leaveTangent=-72.0739, tangentMode=RCTM_User, time=0, value=35}, <at 1 5 RCIM_Cubic>]}}
   }
   n48 = Get LeftMovement() -> (LeftMovement:double)
   n49 = Trail controller - Bone: dyn_jacketFront_04_r!(ComponentPose=@n47, Alpha=@n48, FakeVelocity:Vector="0,0,0") -> (Pose) {
    bActorSpaceFakeVel=true
    bShowTrailLocation=true
    chainLength=3
    debugLifeTime=0.1
    maxDeltaTime=0.0333
    relaxationSpeedScale=3
    alphaScaleBiasClamp {bInterpResult=true}
    baseJoint {boneName=root}
    planarLimits [
     {drivingBone {boneName=root}, planeTransform {location=<0 0 10>, rotation=<rot 0 0 0>, scale=<1 1 1>}}
     {drivingBone {boneName=spine_03}, planeTransform {location=<0 -20 0>, rotation=<rot 0 15 -89.9999>, scale=<1 1 1>}}
    ]
    rotationLimits [{limitMax=<180 180 180>}, {limitMax=<180 180 180>}, {limitMax=<180 180 180>}]
    rotationOffsets [<0 0 0>, <0 0 0>, <0 0 0>]
    trailBone {boneName=dyn_jacketFront_04_r}
    trailRelaxationSpeed {editorCurveData {keys [{arriveTangent=-72.0739, interpMode=RCIM_Cubic, leaveTangent=-72.0739, tangentMode=RCTM_User, time=0, value=35}, <at 1 5 RCIM_Cubic>]}}
   }
   n50 = Trail controller - Bone: dyn_jacketSide_04_r!(ComponentPose=@n49, Alpha=@n48, FakeVelocity:Vector="0,0,0") -> (Pose) {
    bActorSpaceFakeVel=true
    bShowTrailLocation=true
    chainLength=3
    debugLifeTime=0.1
    maxDeltaTime=0.0333
    relaxationSpeedScale=3
    alphaScaleBiasClamp {bInterpResult=true}
    baseJoint {boneName=root}
    planarLimits [
     {drivingBone {boneName=root}, planeTransform {location=<0 0 10>, rotation=<rot 0 0 0>, scale=<1 1 1>}}
     {drivingBone {boneName=spine_03}, planeTransform {location=<0 -20 0>, rotation=<rot 0 15 -89.9999>, scale=<1 1 1>}}
    ]
    rotationLimits [{limitMax=<180 180 180>}, {limitMax=<180 180 180>}, {limitMax=<180 180 180>}]
    rotationOffsets [<0 0 0>, <0 0 0>, <0 0 0>]
    trailBone {boneName=dyn_jacketSide_04_r}
    trailRelaxationSpeed {editorCurveData {keys [{arriveTangent=-72.0739, interpMode=RCIM_Cubic, leaveTangent=-72.0739, tangentMode=RCTM_User, time=0, value=35}, <at 1 5 RCIM_Cubic>]}}
   }
   n51 = Get RightMovement() -> (RightMovement:double)
   n52 = Trail controller - Bone: dyn_jacketFront_04_l!(ComponentPose=@n50, Alpha=@n51, FakeVelocity:Vector="0,0,0") -> (Pose) {
    bActorSpaceFakeVel=true
    bShowTrailLocation=true
    chainLength=3
    debugLifeTime=0.1
    maxDeltaTime=0.0333
    relaxationSpeedScale=3
    alphaScaleBiasClamp {bInterpResult=true}
    baseJoint {boneName=root}
    planarLimits [
     {drivingBone {boneName=root}, planeTransform {location=<0 0 10>, rotation=<rot 0 0 0>, scale=<1 1 1>}}
     {drivingBone {boneName=spine_03}, planeTransform {location=<0 -20 0>, rotation=<rot 0 15 -89.9999>, scale=<1 1 1>}}
    ]
    rotationLimits [{limitMax=<180 180 180>}, {limitMax=<180 180 180>}, {limitMax=<180 180 180>}]
    rotationOffsets [<0 0 0>, <0 0 0>, <0 0 0>]
    trailBone {boneName=dyn_jacketFront_04_l}
    trailRelaxationSpeed {editorCurveData {keys [{arriveTangent=-72.0739, interpMode=RCIM_Cubic, leaveTangent=-72.0739, tangentMode=RCTM_User, time=0, value=35}, <at 1 5 RCIM_Cubic>]}}
   }
   n53 = Trail controller - Bone: dyn_jacketSide_04_l!(ComponentPose=@n52, Alpha=@n51, FakeVelocity:Vector="0,0,0") -> (Pose) {
    bActorSpaceFakeVel=true
    bShowTrailLocation=true
    chainLength=3
    debugLifeTime=0.1
    maxDeltaTime=0.0333
    relaxationSpeedScale=3
    alphaScaleBiasClamp {bInterpResult=true}
    baseJoint {boneName=root}
    planarLimits [
     {drivingBone {boneName=root}, planeTransform {location=<0 0 10>, rotation=<rot 0 0 0>, scale=<1 1 1>}}
     {drivingBone {boneName=spine_03}, planeTransform {location=<0 -20 0>, rotation=<rot 0 15 -89.9999>, scale=<1 1 1>}}
    ]
    rotationLimits [{limitMax=<180 180 180>}, {limitMax=<180 180 180>}, {limitMax=<180 180 180>}]
    rotationOffsets [<0 0 0>, <0 0 0>, <0 0 0>]
    trailBone {boneName=dyn_jacketSide_04_l}
    trailRelaxationSpeed {editorCurveData {keys [{arriveTangent=-72.0739, interpMode=RCIM_Cubic, leaveTangent=-72.0739, tangentMode=RCTM_User, time=0, value=35}, <at 1 5 RCIM_Cubic>]}}
   }
   Save cached pose 'unRigidPose'(Pose=Component To Local(ComponentPose=@n53)) {cacheName=unRigidPose}
  ]}
 EventGraph {nodes [
   Event BlueprintUpdateAnimation()
   > n1 = Sequence!(execute) -> (then_0, then_1, then_2:Exec, then_3:Exec, then_4:Exec, then_5:Exec)
   n2 = GetOwnerSkeletalMeshComponent()
   n3 = BreakRotator(InRot=BreakTransform(InTransform=GetDeltaTransformFromRefPose(self=@n2, BoneName=thigh_r, BaseName=pelvis)).Rotation)
   > Set thigh_rot_up_r(thigh_rot_up_r:double=MapRangeClamped(Value=@n3.Yaw, InRangeA=-15, InRangeB=-100)) -> (Output_Get:double)
   > Set thigh_rot_inr_r(thigh_rot_inr_r:double=MapRangeClamped(Value=@n3.Pitch, InRangeA=0, InRangeB=-10, OutRangeA=1, OutRangeB=0)) -> (Output_Get:double)
   > Set thigh_rot_otr_r(thigh_rot_otr_r:double=MapRangeClamped(Value=@n3.Pitch, InRangeA=20, InRangeB=50)) -> (Output_Get:double)
   n4 = GetOwnerSkeletalMeshComponent()
   n5 = BreakRotator(InRot=BreakTransform(InTransform=GetDeltaTransformFromRefPose(self=@n4, BoneName=thigh_l, BaseName=pelvis)).Rotation)
   > Set thigh_rot_up_l(thigh_rot_up_l:double=MapRangeClamped(Value=@n5.Yaw, InRangeA=-10, InRangeB=-100)) -> (Output_Get:double)
   > Set thigh_rot_inr_l(thigh_rot_inr_l:double=MapRangeClamped(Value=@n5.Pitch, InRangeA=0, InRangeB=-10, OutRangeA=1, OutRangeB=0)) -> (Output_Get:double)
   > Set thigh_rot_otr_l(thigh_rot_otr_l:double=MapRangeClamped(Value=@n5.Pitch, InRangeA=20, InRangeB=50)) -> (Output_Get:double)
   > Set jumpingAlpha(jumpingAlpha:double="To Float (Boolean)"(InBool=Get bIsJumping())) -> (Output_Get:double)
   n6 = Get bIsMoving2D()
   > Set isNotCrouchRunning(isNotCrouchRunning:double="To Float (Boolean)"(InBool=NOT Boolean(A=@n6))) -> (Output_Get:double)
   n7 = Gravity Override State Function DEPRECATED(
    execute=@n1.then_1
    CharacterPartAnimInstance=Self-Reference()
    GameTimeInSeconds=GetGameTimeInSeconds()
    GravityOverrideParams=Get Gravity Override Params()
   )
   > Set GravityOverride(GravityOverride=@n7)
   n8 = ComplexSimSpaceFunction DEPRECATED!(
    execute=@n1.then_2
    CharacterPartAnimInstance:Custom Character Part Anim Instance Object Reference=Self-Reference()
    InterpSpeedMax:double=9
    PawnSpeedMax:double=20
    RInterpErrorTolerance:double=1
    DeltaTime:double=0
    DebugDraw:bool=false
    sim_space_rotation_in:Rotator=Get sim_space_rotation() -> (sim_space_rotation:Rotator)
    inComponentSpaceIn:bool=Get InComponentSpace() -> (InComponentSpace:bool)
    inComponentSpaceLastFrameIn:bool=Get InComponentSpaceLastFrame() -> (InComponentSpaceLastFrame:bool)
    lockedOnSimSpaceIn:bool=Get LockedOnSimSpace() -> (LockedOnSimSpace:bool)
    InterpSpeedIn:double=Get InterpSpeed() -> (InterpSpeed:double)
   ) -> (
    then
    sim_space_rotation:Rotator
    inComponentSpaceOut:bool
    inComponentSpaceLastFrameOut:bool
    lockedOnSimSpaceOut:bool
    InterpSpeedOut:double
   )
   > Set sim_space_rotation(sim_space_rotation:Rotator=@n8) -> (Output_Get:Rotator)
   > Set InComponentSpace(InComponentSpace:bool=@n8.inComponentSpaceOut) -> (Output_Get:bool)
   > Set InComponentSpaceLastFrame(InComponentSpaceLastFrame:bool=@n8.inComponentSpaceLastFrameOut) -> (Output_Get:bool)
   > Set LockedOnSimSpace(LockedOnSimSpace:bool=@n8.lockedOnSimSpaceOut) -> (Output_Get:bool)
   > Set InterpSpeed(InterpSpeed:double=@n8.InterpSpeedOut) -> (Output_Get:double)
   n9 = PlayerMovement(execute=@n1.then_3, AnimInstance=Self-Reference(), XDotProductCoefficient=1, YDotProductCoefficient=1, ZDotProductCoefficient=1)
   n10 = MapRangeClamped(Value=@n9.YMovement, InRangeA=0, InRangeB=-1)
   n11 = float * float(A="Absolute (Float)"(A=@n9), B=0.5)
   > Set BwdMovement(BwdMovement:double=MapRangeClamped(
     Value=float + float(A=@n11, B=@n10) {is=title K2Node_CommutativeAssociativeBinaryOperator}
     InRangeA=0
     InRangeB=1.1
     OutRangeA=0.4
     OutRangeB=0
    )) -> (Output_Get:double)
   n12 = MapRangeClamped(Value=@n9, InRangeA=-0.2, InRangeB=1)
   n13 = float * float(A="Absolute (Float)"(A=@n9.YMovement), B=0.2)
   > Set RightMovement(RightMovement:double=MapRangeClamped(Value=float + float(A=@n13, B=@n12), InRangeA=0, InRangeB=1.2, OutRangeB=0.5)) -> (Output_Get:double)
   n14 = MapRangeClamped(Value=@n9, InRangeA=0.2, InRangeB=-1, OutRangeB=0.3)
   n15 = float * float(A="Absolute (Float)"(A=@n9.YMovement), B=0.2)
   > Set LeftMovement(LeftMovement:double=MapRangeClamped(Value=float + float(A=@n15, B=@n14), InRangeA=0, InRangeB=1.2, OutRangeB=0.5)) -> (Output_Get:double)
   n16 = Branch(execute=@n1.then_4, Condition=Get bIsSkyDiving())
   > n17 = Branch(Condition="float > float"(A=Get PawnVelocityZ(), B=50) {is=title K2Node_CallFunction})
   > Set BwdMovement(BwdMovement=float * float(A=Get BwdMovement(), B=0))
   Set BwdMovement(execute=[@n16.else, @n17.else], BwdMovement=float * float(A=Get BwdMovement(), B=1))
   Get SlopeInfoInput() -> (SlopeInfoInput:Fort Anim Input Slope Info Structure)
   Get bIsCrouchSprinting()
  ]}
]
variables [
 BwdMovement {type=number}
 Gravity Override Params {
  type=GravityOverrideParamsStruct
  default {
   skydive_FloatingDown {gravityOverride=<0 -700 100>, windAmplitude=<200 -500 200>, windOffset=<0 -600 200>}
   skydive_GoingUp {gravityOverride=<0 -300 500>, windAmplitude=<200 -500 300>}
   skydive_Parachuting {gravityOverride=<0 -500 -980>, windAmplitude=<200 -300 400>}
  }
 }
 GravityOverride {type=Vector}
 InComponentSpace {type=boolean}
 InComponentSpaceLastFrame {type=boolean}
 InterpSpeed {type=number}
 isNotCrouchRunning {type=number}
 jumpingAlpha {type=number}
 LeftMovement {type=number}
 LockedOnSimSpace {type=boolean}
 RightMovement {type=number}
 sim_space_rotation {type=Rotator}
 thigh_rot_inr_l {type=number}
 thigh_rot_inr_r {type=number}
 thigh_rot_otr_l {type=number}
 thigh_rot_otr_r {type=number}
 thigh_rot_up_l {type=number}
 thigh_rot_up_r {type=number}
]
