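# Physics asset of capsule colliders for the Psyche male medium player body, named for cape collision.
# One capsule per bone covers the head, neck, spine, pelvis, shoulders, upper arms, legs and feet, with no forearms or hands.
# An extra large capsule sits on a dyn_crouch_collision bone, apparently for crouch poses, and has no constraint.
# Constraints link the other bodies along the skeleton with all swing and twist motion free.
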PhysicsAsset /BRCosmetics/Characters/Player/Male/Medium/Bodies/M_MED_Psyche/Meshes/M_MED_Psyche_CapeColliders
bodies {
 calf_l [{shapeType=Capsule, length=28.742, radius=20.1616, center=<-24.5773 -4.1475 -0.2791>, rotation=<rot 0.0927 84.0793 -89.474>}]
 calf_r [{shapeType=Capsule, length=28.742, radius=20.1616, center=<24.5773 4.1475 0.2791>, rotation=<rot 0.0927 84.0793 90.526>}]
 clavicle_l [{shapeType=Capsule, length=10.3261, radius=8.5802, center=<8.2822 -0.2527 -2.0684>, rotation=<rot 90 -0.0049 179.9951>}]
 clavicle_r [{shapeType=Capsule, length=10.3261, radius=8.5802, center=<-7.9705 0.6201 2.6048>, rotation=<rot 90 0.0198 0.0198>}]
 dyn_crouch_collision [{
   shapeType=Capsule
   length=29.25
   radius=36.25
   shapeName=dyn_crouch_collision_capsule
   center=<0 -15.964 7.103>
   rotation=<rot 90 -0.0396 -0.0396>
  }]
 foot_l [{
   shapeType=Capsule
   length=23.9959
   radius=18.5957
   shapeName=foot_l_capsule
   center=<-4.2543 6.5352 0.1751>
   rotation=<rot 0 0.2297 -87.3073>
  }]
 foot_r [{
   shapeType=Capsule
   length=23.9959
   radius=18.5957
   shapeName=foot_l_capsule
   center=<4.2543 -6.5352 -0.1751>
   rotation=<rot 0 0.2297 92.6927>
  }]
 head [{shapeType=Capsule, length=0, radius=18, center=<11.6361 0.0962 0>, rotation=<rot 0 -90 0>}]
 neck_02 [{shapeType=Capsule, length=0.505, radius=10.755, rotation=<rot 0 -90 0>}]
 pelvis [{shapeType=Capsule, length=12.4559, radius=31.5829, center=<-38.0551 -5.1275 1.0153>, rotation=<rot 90 0 -4.9999>}]
 spine_03 [{shapeType=Capsule, length=10.5169, radius=16.5072, center=<-1.4056 -1.8302 0.4275>, rotation=<rot -83.7243 -83.7209 89.9955>}]
 spine_05 [{shapeType=Capsule, length=16.7232, radius=12.517, center=<-2.3724 -0.6607 0>}]
 thigh_l [{shapeType=Capsule, length=29.1153, radius=19.1815, center=<-22.8104 -9.5699 -0.0001>, rotation=<rot 90 0 164.571>}]
 thigh_r [{shapeType=Capsule, length=29.1153, radius=19.1815, center=<22.8104 9.5699 -0.0001>, rotation=<rot 90 0 -15.4288>}]
 upperarm_l [{shapeType=Capsule, length=18.8261, radius=7.5802, center=<12.1824 -0.2526 -0.6865>, rotation=<rot 90 0.005 -179.995>}]
 upperarm_r [{shapeType=Capsule, length=17.1357, radius=7.1339, center=<-12.8316 -0.0074 -0.1319>, rotation=<rot 0 -88.4734 -86.4203>}]
}
constraints [
 {
  bone1Name=spine_03
  bone2Name=pelvis
  swing1LimitDegrees=45
  swing1Motion=Free
  swing2LimitDegrees=45
  swing2Motion=Free
  twistLimitDegrees=45
  twistMotion=Free
 }
 {
  bone1Name=spine_05
  bone2Name=spine_03
  swing1LimitDegrees=45
  swing1Motion=Free
  swing2LimitDegrees=45
  swing2Motion=Free
  twistLimitDegrees=45
  twistMotion=Free
 }
 {
  bone1Name=clavicle_l
  bone2Name=spine_05
  swing1LimitDegrees=45
  swing1Motion=Free
  swing2LimitDegrees=45
  swing2Motion=Free
  twistLimitDegrees=45
  twistMotion=Free
 }
 {
  bone1Name=upperarm_l
  bone2Name=clavicle_l
  swing1LimitDegrees=45
  swing1Motion=Free
  swing2LimitDegrees=45
  swing2Motion=Free
  twistLimitDegrees=45
  twistMotion=Free
 }
 {
  bone1Name=clavicle_r
  bone2Name=spine_05
  swing1LimitDegrees=45
  swing1Motion=Free
  swing2LimitDegrees=45
  swing2Motion=Free
  twistLimitDegrees=45
  twistMotion=Free
 }
 {
  bone1Name=upperarm_r
  bone2Name=clavicle_r
  swing1LimitDegrees=45
  swing1Motion=Free
  swing2LimitDegrees=45
  swing2Motion=Free
  twistLimitDegrees=45
  twistMotion=Free
 }
 {
  bone1Name=thigh_r
  bone2Name=pelvis
  swing1LimitDegrees=45
  swing1Motion=Free
  swing2LimitDegrees=45
  swing2Motion=Free
  twistLimitDegrees=45
  twistMotion=Free
 }
 {
  bone1Name=head
  bone2Name=neck_02
  swing1LimitDegrees=45
  swing1Motion=Free
  swing2LimitDegrees=45
  swing2Motion=Free
  twistLimitDegrees=45
  twistMotion=Free
 }
 {
  bone1Name=neck_02
  bone2Name=spine_05
  swing1LimitDegrees=45
  swing1Motion=Free
  swing2LimitDegrees=45
  swing2Motion=Free
  twistLimitDegrees=45
  twistMotion=Free
 }
 {
  bone1Name=calf_r
  bone2Name=thigh_r
  swing1LimitDegrees=45
  swing1Motion=Free
  swing2LimitDegrees=45
  swing2Motion=Free
  twistLimitDegrees=45
  twistMotion=Free
 }
 {
  bone1Name=foot_r
  bone2Name=calf_r
  swing1LimitDegrees=45
  swing1Motion=Free
  swing2LimitDegrees=45
  swing2Motion=Free
  twistLimitDegrees=45
  twistMotion=Free
 }
 {
  bone1Name=foot_l
  bone2Name=calf_l
  swing1LimitDegrees=45
  swing1Motion=Free
  swing2LimitDegrees=45
  swing2Motion=Free
  twistLimitDegrees=45
  twistMotion=Free
 }
 {
  bone1Name=calf_l
  bone2Name=thigh_l
  swing1LimitDegrees=45
  swing1Motion=Free
  swing2LimitDegrees=45
  swing2Motion=Free
  twistLimitDegrees=45
  twistMotion=Free
 }
 {
  bone1Name=thigh_l
  bone2Name=pelvis
  swing1LimitDegrees=45
  swing1Motion=Free
  swing2LimitDegrees=45
  swing2Motion=Free
  twistLimitDegrees=45
  twistMotion=Free
 }
]
