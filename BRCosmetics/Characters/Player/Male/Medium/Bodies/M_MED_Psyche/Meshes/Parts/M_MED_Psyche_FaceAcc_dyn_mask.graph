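# Physics asset for the dangling mask straps and tubes on the Psyche male medium face accessory.
# Capsule bodies form a left three-link and right four-link mask chain plus left and right tubes, alongside larger head and upper-spine capsules.
# Each chain hangs from a small master or aim body with free rotation relative to the head.
# Links use limited swing and tight twist, with tubes more restricted on one swing axis.
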PhysicsAsset /BRCosmetics/Characters/Player/Male/Medium/Bodies/M_MED_Psyche/Meshes/Parts/M_MED_Psyche_FaceAcc_dyn_mask
bodies {
 dyn_mask_l_01 [{shapeType=Capsule, length=1, radius=1.5, shapeName=dyn_mask_l_01_capsule, center=<1 0 0>, rotation=<rot 90 0 0>}]
 dyn_mask_l_02 [{shapeType=Capsule, length=2, radius=1.5, shapeName=dyn_mask_l_02_capsule, center=<2 0 0>, rotation=<rot 90 0 0>}]
 dyn_mask_l_03 [{shapeType=Capsule, length=4.75, radius=1.5, shapeName=dyn_mask_l_03_capsule, center=<3 0 0>, rotation=<rot 90 0 0>}]
 dyn_mask_l_MSTR [{shapeType=Capsule, length=0.505, radius=0.505, shapeName=dyn_mask_l_MSTR_capsule, rotation=<rot 0 -90 0>}]
 dyn_mask_r_01 [{shapeType=Capsule, length=1, radius=1.5, shapeName=dyn_mask_r_01_capsule, center=<1 0 0>, rotation=<rot 90 0 0>}]
 dyn_mask_r_02 [{shapeType=Capsule, length=2, radius=1.5, shapeName=dyn_mask_r_02_capsule, center=<2 0 0>, rotation=<rot 90 0 0>}]
 dyn_mask_r_03 [{shapeType=Capsule, length=2, radius=1.5, shapeName=dyn_mask_r_03_capsule, center=<2 0 0>, rotation=<rot 90 0 0>}]
 dyn_mask_r_04 [{shapeType=Capsule, length=2, radius=1.5, shapeName=dyn_mask_r_04_capsule, center=<2 0 0>, rotation=<rot 90 0 0>}]
 dyn_mask_r_MSTR [{shapeType=Capsule, length=0.505, radius=0.505, shapeName=dyn_mask_r_MSTR_capsule, rotation=<rot 0 -90 0>}]
 dyn_tube_l [{shapeType=Capsule, length=2, radius=2, shapeName=dyn_tube_l_capsule, center=<1 0 0>, rotation=<rot 90 0 0>}]
 dyn_tube_l_AIM [{shapeType=Capsule, length=0.505, radius=0.505, shapeName=dyn_mask_l_MSTR_capsule, rotation=<rot 0 -90 0>}]
 dyn_tube_r [{shapeType=Capsule, length=2, radius=2, shapeName=dyn_tube_r_capsule, center=<1 0 0>, rotation=<rot 90 0 0>}]
 dyn_tube_r_AIM [{shapeType=Capsule, length=0.505, radius=0.505, shapeName=dyn_mask_l_MSTR_capsule, rotation=<rot 0 -90 0>}]
 head_offset [{shapeType=Capsule, length=5.7812, radius=8.0341, center=<4.6174 1.15 -0.2528>, rotation=<rot 90 0 0>}]
 spine_05 [{shapeType=Capsule, length=18.2169, radius=11.0419, center=<-0.1031 -1.3094 0.0248>}]
}
constraints [
 {bone1Name=head_offset, bone2Name=spine_05, swing1LimitDegrees=45, swing2LimitDegrees=45, twistLimitDegrees=45}
 {
  bone1Name=dyn_mask_l_MSTR
  bone2Name=head_offset
  swing1LimitDegrees=45
  swing1Motion=Free
  swing2LimitDegrees=45
  swing2Motion=Free
  twistLimitDegrees=45
  twistMotion=Free
 }
 {bone1Name=dyn_mask_l_01, bone2Name=dyn_mask_l_MSTR, swing1LimitDegrees=25, swing2LimitDegrees=25, twistLimitDegrees=2}
 {bone1Name=dyn_mask_l_02, bone2Name=dyn_mask_l_01, swing1LimitDegrees=25, swing2LimitDegrees=25, twistLimitDegrees=2}
 {bone1Name=dyn_mask_l_03, bone2Name=dyn_mask_l_02, swing1LimitDegrees=25, swing2LimitDegrees=25, twistLimitDegrees=2}
 {
  bone1Name=dyn_mask_r_MSTR
  bone2Name=head_offset
  swing1LimitDegrees=45
  swing1Motion=Free
  swing2LimitDegrees=45
  swing2Motion=Free
  twistLimitDegrees=45
  twistMotion=Free
 }
 {bone1Name=dyn_mask_r_01, bone2Name=dyn_mask_r_MSTR, swing1LimitDegrees=25, swing2LimitDegrees=25, twistLimitDegrees=2}
 {bone1Name=dyn_mask_r_02, bone2Name=dyn_mask_r_01, swing1LimitDegrees=25, swing2LimitDegrees=25, twistLimitDegrees=2}
 {bone1Name=dyn_mask_r_03, bone2Name=dyn_mask_r_02, swing1LimitDegrees=25, swing2LimitDegrees=25, twistLimitDegrees=2}
 {bone1Name=dyn_mask_r_04, bone2Name=dyn_mask_r_03, swing1LimitDegrees=25, swing2LimitDegrees=25, twistLimitDegrees=2}
 {
  bone1Name=dyn_tube_l_AIM
  bone2Name=head_offset
  swing1LimitDegrees=45
  swing1Motion=Free
  swing2LimitDegrees=45
  swing2Motion=Free
  twistLimitDegrees=45
  twistMotion=Free
 }
 {bone1Name=dyn_tube_l, bone2Name=dyn_tube_l_AIM, swing1LimitDegrees=5, swing2LimitDegrees=20, twistLimitDegrees=2}
 {
  bone1Name=dyn_tube_r_AIM
  bone2Name=head_offset
  swing1LimitDegrees=45
  swing1Motion=Free
  swing2LimitDegrees=45
  swing2Motion=Free
  twistLimitDegrees=45
  twistMotion=Free
 }
 {bone1Name=dyn_tube_r, bone2Name=dyn_tube_r_AIM, swing1LimitDegrees=5, swing2LimitDegrees=20, twistLimitDegrees=2}
]
