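# Animation Blueprint for the face accessory of the M_MED_Psyche male character, driving its mask and tube bones.
# Copies the attached parent mesh's pose, partially copies head rotation onto head_offset, and aims both tube bones at target bones.
# With bRigidBodyEnabled set, a RigidBody node simulates a mask physics asset in neck_01 space with damped component motion.
# Otherwise, Look At nodes aim mask bones toward upper spine bones as a non-simulated fallback.
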AnimBlueprint /BRCosmetics/Characters/Player/Male/Medium/Bodies/M_MED_Psyche/Meshes/Parts/M_MED_Psyche_FaceAcc_AnimBP
target_skeleton=&/M_MED_Psyche_FaceAcc_Skeleton
graphs [
 AnimGraph {nodes [
   Output Pose(Result=Blend Poses by bool(BlendPose_0=Use cached pose 'rigidPose'(), BlendPose_1=Use cached pose 'unRigidPose'(), bActiveValue=Get bRigidBodyEnabled()) {blendType=Linear}) {layerGroup=None}
   n1 = Copy Bone - Source Bone: head - Target Bone: head_offset!(
    ComponentPose:Component Space Pose Link Structure=Local To Component(LocalPose=Copy Pose From Mesh() {bUseAttachedParent=true})
    Alpha:float=0.6
    bCopyTranslation:bool=False
    bCopyRotation:bool=true
    bCopyScale:bool=False
   ) -> (Pose:Component Space Pose Link Structure) {
    is=title AnimGraphNode_CopyBone
    controlSpace=BCS_BoneSpace
    alphaBoolBlend {blendOption=Linear}
    sourceBone {boneName=head}
    targetBone {boneName=head_offset}
   }
   n2 = Look At - Bone: dyn_tube_l_AIM(ComponentPose=@n1, Alpha=1) {
    alphaBoolBlend {blendOption=Linear}
    boneToModify {boneName=dyn_tube_l_AIM}
    lookAtLocation {x=0}
    lookAtTarget {boneReference {boneName=dyn_tube_l_TGT}}
    lookAt_Axis {axis=<1 0>}
    lookUp_Axis {axis {x=0}}
   }
   n3 = Look At - Bone: dyn_tube_r_AIM(ComponentPose=@n2, Alpha=1) {
    alphaBoolBlend {blendOption=Linear}
    boneToModify {boneName=dyn_tube_r_AIM}
    lookAtLocation {x=0}
    lookAtTarget {boneReference {boneName=dyn_tube_r_TGT}}
    lookAt_Axis {axis=<1 0>}
    lookUp_Axis {axis {x=0}}
   }
   Save cached pose 'prePose'(Pose=Component To Local(ComponentPose=@n3)) {cacheName=prePose}
   n4 = RigidBody!(ComponentPose=Local To Component(LocalPose=Use cached pose 'prePose'()), Alpha, OverrideWorldGravity:Vector="0,0,-980", ExternalForce) -> (Pose) {
    bClampLinearTranslationLimitToRefPose=true
    bOverrideWorldGravity=true
    overridePhysicsAsset=&/M_MED_Psyche_FaceAcc_dyn_mask
    simulationSpace=BaseBoneSpace
    alphaBoolBlend {blendOption=Linear}
    baseBoneRef {boneName=neck_01}
    componentAppliedLinearAccClamp=<1000 1000 1000>
    componentLinearAccScale=<0.25 0.25 0.25>
    componentLinearVelScale=<0.25 0.25 0.25>
    overrideWorldGravity {z=-980}
    simSpaceSettings {maxAngularAcceleration=200, maxAngularVelocity=2, maxLinearAcceleration=800, maxLinearVelocity=400, velocityScaleZ=0}
   }
   Save cached pose 'rigidPose'(Pose=Component To Local(ComponentPose=@n4)) {cacheName=rigidPose}
   n5 = Look At - Bone: dyn_mask_r_01(ComponentPose=Local To Component(LocalPose=Use cached pose 'prePose'()), Alpha=1) {
    alphaBoolBlend {blendOption=Linear}
    boneToModify {boneName=dyn_mask_r_01}
    lookAtLocation=<-5 11.15 3>
    lookAtTarget {boneReference {boneName=spine_05}}
    lookAt_Axis {axis=<1 0>}
   }
   n6 = Look At - Bone: dyn_mask_r_03(ComponentPose=@n5, Alpha=1) {
    alphaBoolBlend {blendOption=Linear}
    boneToModify {boneName=dyn_mask_r_03}
    lookAtLocation=<-10 11.15 3>
    lookAtTarget {boneReference {boneName=spine_05}}
    lookAt_Axis {axis=<1 0>}
   }
   n7 = Look At - Bone: dyn_mask_r_04(ComponentPose=@n6, Alpha=1) {
    alphaBoolBlend {blendOption=Linear}
    boneToModify {boneName=dyn_mask_r_04}
    lookAtLocation=<0 11.15 3>
    lookAtTarget {boneReference {boneName=spine_04}}
    lookAt_Axis {axis=<1 0>}
   }
   n8 = Look At - Bone: dyn_mask_l_01(ComponentPose=@n7, Alpha=1) {
    alphaBoolBlend {blendOption=Linear}
    boneToModify {boneName=dyn_mask_l_01}
    lookAtLocation=<-5 11 -3>
    lookAtTarget {boneReference {boneName=spine_05}}
    lookAt_Axis {axis=<1 0>}
   }
   n9 = Look At - Bone: dyn_mask_l_03(ComponentPose=@n8, Alpha=1) {
    alphaBoolBlend {blendOption=Linear}
    boneToModify {boneName=dyn_mask_l_03}
    lookAtLocation=<-10 11 -3>
    lookAtTarget {boneReference {boneName=spine_05}}
    lookAt_Axis {axis=<1 0>}
   }
   Save cached pose 'unRigidPose'(Pose=Component To Local(ComponentPose=@n9)) {cacheName=unRigidPose}
  ]}
 EventGraph {nodes [Event BlueprintUpdateAnimation()]}
]
variables []
